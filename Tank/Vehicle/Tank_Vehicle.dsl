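; Fortnite drivable tank vehicle Blueprint, handling damageable parts, cannon recoil, drive-over damage, water FX and explosion.
; Timers drive engine overheat, delayed self-repair and driver-cover cooldown, feeding damage Niagara, audio controller and broadcasts.
; Health thresholds trigger loot drops and squad quest events; critical health can ignite Curie fire.
; Explosion launches the tank, then the server applies GameplayEffect damage and impulses to nearby line-of-sight actors and passengers.
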(event Custom|TankTreadRepaired
  (bind _returnvalue (Utilities|Time|GetGameTimeinSeconds))
  (bind _skeletalmesh (Variables|Vehicle|GetSkeletalMesh))
  ;[1 0.6177 0] EVENT -> Tread Repaired
  (Vehicle|WakeUp)
  (for _array_element (Variables|/Script/FortniteGame.FortVehicleConfigs|GetDamageableParts (Class|FortVehicleInterface|GetVehicleConfigs))
    (switch name (Utilities|Struct|BreakVehicleDamageablePartConfig _array_element)
      (:Tread_L
        (CallFunction|RepairDamageablePart (Utilities|Struct|BreakVehicleDamageablePartConfig _array_element))
        (Rendering|Material|SetCustomPrimitiveDataFloat _skeletalmesh 7 _returnvalue))
      (:Tread_R
        (CallFunction|RepairDamageablePart (Utilities|Struct|BreakVehicleDamageablePartConfig _array_element))
        (Rendering|Material|SetCustomPrimitiveDataFloat _skeletalmesh 6 _returnvalue)))
    (Class|BPTankAudioController|TreadsRepairAudio (Utilities|GetAudioController) "None")))

(event Custom|TankEngineHealthChanged
  (params (CurrentHealth :type Float%20%28double-precision%29))
  (bind _enginemaxhealth (Variables|TankEngine|GetEngineMaxHealth))
  (bind _returnvalue (Math|Float|SafeDivide CurrentHealth _enginemaxhealth))
  (bind _tankenginerepairdelaytimer (Variables|TankEngine|GetTankEngineRepairDelayTimer))
  (bind _tankenginerepairtimer (Variables|TankEngine|GetTankEngineRepairTimer))
  ;[0.2812 0.4729 1] EVENT -> On Engine Damage and Repair
  (if (Variables|VehicleExplode|GetVehicleExploded)
    (bind _niagaracomp_enginedamage (Variables|Default|GetNiagaraComp_EngineDamage)
      (:then
        (Components|Activation|Deactivate _niagaracomp_enginedamage)
        (Components|DestroyComponent _niagaracomp_enginedamage))
      (:else))
    (elif (or (Variables|TankEngine|GetEngineOverheated) (>= (Variables|TankEngine|GetTankEngineCurrent%20Health) (Variables|TankEngine|GetEngineMaxHealth)))
      _
      (elif (> CurrentHealth (Variables|TankEngine|GetTankEnginePreviousHealth))
        (Variables|TankEngine|SetTankEnginePreviousHealth CurrentHealth)
        (CallFunction|UpdateEngineDamageState _returnvalue)
        (else
          (Variables|TankEngine|SetTankEnginePreviousHealth CurrentHealth)
          (Utilities|FlowControl|Sequence
            (:then_0
              (if (Utilities|Time|IsValidTimerHandle _tankenginerepairtimer)
                (Utilities|Time|ClearandInvalidateTimerbyHandle _tankenginerepairtimer))
              (if (Utilities|Time|IsValidTimerHandle _tankenginerepairdelaytimer)
                (Utilities|Time|ClearandInvalidateTimerbyHandle _tankenginerepairdelaytimer) ; If already running when we take damage, reset the timer
              )
              (bind _returnvalue_2 (Utilities|Time|SetTimerbyEvent (EventDispatchers|CreateEvent :!event Custom|BeginRepairTankEngineTimer) (Variables|TankEngine|GetDelayToBeginEngineRepairOnDamage)))
              (Variables|TankEngine|SetTankEngineRepairDelayTimer _returnvalue_2))
            (:then_1
              (CallFunction|UpdateEngineDamageState _returnvalue))))))))

(event Custom|EventOnEngineOverheat
  (bind _niagaracomp_enginedamage (Variables|Default|GetNiagaraComp_EngineDamage))
  ;[0.2812 0.4729 1] EVENT ->  Engine Overheat / Cooldown
  (Variables|TankEngine|SetEngineOverheated true)
  (Variables|TankEngine|SetTankEngineCurrentHealthPercent 0.0)
  (bind _returnvalue (Utilities|Time|SetTimerbyEvent (delegate Custom|EventOnEngineCooledDown) (Variables|TankEngine|GetEngineOverheatDuration)))
  (Variables|TankEngine|SetEngineOverheatDurationTimer _returnvalue)
  (Class|BPTankAudioController|EngineOverheatedAudio (Utilities|GetAudioController))
  (if (Variables|VehicleExplode|GetVehicleExploded)
    (bind _niagaracomp_enginedamage_2 (Variables|Default|GetNiagaraComp_EngineDamage)
      (:then
        (Components|DestroyComponent _niagaracomp_enginedamage_2))
      (:else)) ; If we get destroyed after the vehicle explodes destroy the FX
    (else
      (CallFunction|UpdateEngineDamageState 0.0)
      (Default|CallEngineEnabledChangedBroadcast false)
      (Utilities|IsValid _niagaracomp_enginedamage
        (:Is%20Valid
          (Niagara|SetNiagaraVariable(Float) _niagaracomp_enginedamage "User.DamageScale" 1.0))
        (:Is%20Not%20Valid
          (bind _returnvalue_2 (Niagara|SpawnSystemAttached (Variables|TankEngine|GetNiagaraSystemEngineDamage) (Variables|Vehicle|GetSkeletalMesh) "Engine" "0, 0, 0" "0, 0, 0" "KeepRelativeOffset" false true "AutoRelease"))
          (Variables|Default|SetNiagaraComp_EngineDamage _returnvalue_2))) ; Still need to scale FX if it's active and we are healed beyond min threshold
      (Utilities|Time|SetTimerbyEvent (delegate Custom|EjectPlayersOnEngineOverheat) (Variables|TankEngine|GetDelayToEjectPlayersOnEngineOverheat)))))

(event Custom|StopEngineRepairTimers
  ;[0.2812 0.4729 1] EVENT -> On Engine Damage and Repair
  (Utilities|Time|ClearandInvalidateTimerbyHandle (Variables|TankEngine|GetTankEngineRepairTimer))
  (Utilities|Time|ClearandInvalidateTimerbyHandle (Variables|TankEngine|GetTankEngineRepairDelayTimer)))

(event Custom|EventSentQuestEventOnDamageThreshold
  (params (DamagingPlayerController :type Fort%20Player%20Controller%20Object%20Reference) (Hit :type Hit%20Result%20Structure))
  ;[1 1 1] EVENT -> On Damage Threshold (Drops loot and quest items)
  (bind _asfort_player_controller_athena (Utilities|Casting|CastToFortPlayerControllerAthena DamagingPlayerController))
  (for _array_element (Team|GetSquadMemberPlayerStates _asfort_player_controller_athena true)
    (Ability|SendGameplayEventtoActor (PlayerState|GetCurrentPawn _array_element) (Variables|Quests|GetTag_QuestOnDamageThreshold) (Utilities|Struct|MakeGameplayEventData (Variables|Quests|GetTag_QuestOnDamageThreshold) (PlayerState|GetCurrentPawn _array_element) self 0 0 "()" (GameplayTags|GetOwnedGameplayTags (PlayerState|GetCurrentPawn _array_element) :!class /Script/GameplayTags.GameplayTagAssetInterface) (Class|FortVehicleInterface|GetGameplayTags self) 0.0 (Ability|TargetData|AbilityTargetDatafromHitResult Hit)))))

(event Custom|EventOnEngineCooledDown
  ;[0.2812 0.4729 1] EVENT ->  Engine Overheat / Cooldown
  (CallFunction|RepairDamageablePart (Variables|TankEngine|GetEngineDamageablePart))
  (Variables|TankEngine|SetTankEngineCurrent%20Health (Variables|TankEngine|GetEngineMaxHealth))
  (Variables|TankEngine|SetEngineOverheated false)
  (Variables|TankEngine|SetTankEngineCurrentHealthPercent 1.0)
  (CallFunction|UpdateEngineDamageState 1.0)
  (Class|BPTankAudioController|EngineOverheatedAudio (Utilities|GetAudioController))
  (Vehicle|RemoveReasonWhyCantContainPlayers (Utilities|Text|MakeLiteralText (Variables|TankEngine|GetEngineOverheatedMessage))))

(event Custom|EjectPlayersOnEngineOverheat
  ;[0.2812 0.4729 1] EVENT ->  Engine Overheat / Cooldown
  (Vehicle|AddReasonWhyCantContainPlayers (Utilities|Text|MakeLiteralText (Variables|TankEngine|GetEngineOverheatedMessage)))
  (CallFunction|StopEngineRepairTimers))

(event Custom|BeginRepairTankEngineTimer
  ;[0.2812 0.4729 1] EVENT -> On Engine Damage and Repair
  (bind _returnvalue (Utilities|Time|SetTimerbyEvent (delegate Custom|RepairTankEngine) (Variables|TankEngine|GetTankEngineRepairTickRate) true))
  (Variables|TankEngine|SetTankEngineRepairTimer _returnvalue))

(event Custom|EventDropLootOnDamageThreshold
  (params (DamagingActor :type Actor%20Object%20Reference) (Hit :type Hit%20Result%20Structure%20%28by%20ref%29))
  (bind (_bblockinghit _binitialoverlap _time _distance _location _impactpoint _normal) (Collision|BreakHitResult Hit))
  ;[1 1 1] EVENT -> On Damage Threshold (Drops loot and quest items)
  (if (Class|AthenaFunctionLibrary|RowtoBool (Variables|LootDropOnDamage|GetRow_ShouldDropLootOnDamage))
    (bind _returnvalue (TargetSelection|PushOffHitLocation self Hit (Variables|LootDropOnDamage|GetLootTossOffset)))
    (bind _itempickupsspawned (Class|AthenaFunctionLibrary|SpawnLootfromTable (Variables|LootDropOnDamage|GetLootOnDamagePackage) 1 -1 _returnvalue _normal 0 true true)) ; Roll a loot package and toss it in the direction of the player
    (Effects|SpawnEmitteratLocation (Variables|FX|GetFX_Cascade_LootDropOnDamagedSpawn) _returnvalue (Math|Conversions|RotationFromXVector _normal) "1.000000,1.000000,1.000000" true "AutoRelease")
    (for _array_element _itempickupsspawned
      (Pickup|TossPickup _array_element (Math|Vector|SelectVector (+ _location (* (Math|Vector|Normalize (- (Transformation|GetActorLocation DamagingActor) _location)) (Variables|LootDropOnDamage|GetLootTossMin))) (Transformation|GetActorLocation DamagingActor) (>= (Math|Vector|VectorLength (- (Transformation|GetActorLocation DamagingActor) _location)) (Variables|LootDropOnDamage|GetLootTossMin))) 0 0 true false "Tossed" "Unset"))))

(event Custom|UpdateEngineDamageState
  (params (CurrentHealthPercent :type Float%20%28double-precision%29 :default 5e+02))
  (bind _returnvalue (- 1.0 CurrentHealthPercent))
  (bind _niagaracomp_enginedamage (Variables|Default|GetNiagaraComp_EngineDamage))
  (bind _niagaracomp_enginedamage_2 (Variables|Default|GetNiagaraComp_EngineDamage))
  ;[0.2812 0.4729 1] EVENT -> On Engine Damage and Repair
  ;;[0 0 0] Engine Damage State Cosmetics
  (Class|BPTankAudioController|EngineDamageAudio (Utilities|GetAudioController) CurrentHealthPercent)
  (if (Variables|VehicleExplode|GetVehicleExploded)
    (bind _niagaracomp_enginedamage_3 (Variables|Default|GetNiagaraComp_EngineDamage)
      (:then
        (Components|Activation|Deactivate _niagaracomp_enginedamage_3)
        (Components|DestroyComponent _niagaracomp_enginedamage_3))
      (:else))
    (elif (<= CurrentHealthPercent (Variables|TankEngine|GetEngineHealthPercentToEnableEffects))
      (Utilities|IsValid _niagaracomp_enginedamage_2
        (:Is%20Valid
          _)
        (:Is%20Not%20Valid
          (bind _returnvalue_2 (Niagara|SpawnSystemAttached (Variables|TankEngine|GetNiagaraSystemEngineDamage) (Variables|Vehicle|GetSkeletalMesh) "Engine" "0, 0, 0" "0, 0, 0" "KeepRelativeOffset" true true "AutoRelease"))
          (Variables|Default|SetNiagaraComp_EngineDamage _returnvalue_2))) ; Check if FX are enabled. If not valid -> Spawn System
      (Niagara|SetNiagaraVariable(Float) _niagaracomp_enginedamage_2 "User.DamageScale" _returnvalue)
      (else
        (Utilities|IsValid _niagaracomp_enginedamage
          (:Is%20Valid
            (Niagara|SetNiagaraVariable(Float) _niagaracomp_enginedamage "User.DamageScale" _returnvalue))) ; Still need to scale FX if it's active and we are healed beyond min threshold
      ))))

(event Custom|RepairTankEngine
  (bind _tankenginecurrent_health (Variables|TankEngine|GetTankEngineCurrent%20Health))
  (bind _enginemaxhealth (Variables|TankEngine|GetEngineMaxHealth))
  ;[0.2812 0.4729 1] EVENT -> On Engine Damage and Repair
  (if (>= _tankenginecurrent_health _enginemaxhealth)
    (CallFunction|StopEngineRepairTimers)
    (Default|CallEngineEnabledChangedBroadcast true)
    (else
      (CallFunction|RepairDamageablePart (Variables|TankEngine|GetEngineDamageablePart) (+ (Math|Float|SafeDivide _tankenginecurrent_health _enginemaxhealth) (Variables|TankEngine|GetTankEngineRepairPerTick))))))

(event Custom|EventOnDamageThreshold
  (params (DamagingActor :type Actor%20Object%20Reference) (HitResult :type Hit%20Result%20Structure) (DamagingPlayerController :type Fort%20Player%20Controller%20Object%20Reference))
  (bind _returnvalue (Class|FortVehicleInterface|GetHealth))
  (bind _returnvalue_2 (Class|FortVehicleInterface|GetMaxHealth))
  (bind _returnvalue_3 (Math|Float|Max(Float) _returnvalue_2 0.1))
  (bind _returnvalue_4 (/ _returnvalue _returnvalue_3))
  (bind _lootdropdamagethresholds (Variables|LootDropOnDamage|GetLootDropDamageThresholds))
  (bind _returnvalue_5 (ScalableFloat|GetValueatLevel _lootdropdamagethresholds))
  (bind _returnvalue_6 (Math|Float|Max(Float) _returnvalue_5 0.1))
  (bind _returnvalue_7 (/ _returnvalue_4 _returnvalue_6))
  (bind _returnvalue_8 (* _returnvalue_7 1e+02))
  (bind _returnvalue_9 (Math|Float|Floor _returnvalue_8))
  (bind _questeventdamagepercentthreshold (Variables|Quests|GetQuestEventDamagePercentThreshold))
  (bind _returnvalue_10 (Class|FortVehicleInterface|GetMaxHealth))
  (bind _returnvalue_11 (Math|Float|Max(Float) _returnvalue_10 0.1))
  (bind _returnvalue_12 (Class|FortVehicleInterface|GetHealth))
  (bind _returnvalue_13 (/ _returnvalue_12 _returnvalue_11))
  (bind _returnvalue_14 (Math|Float|SafeDivide _returnvalue_13 _questeventdamagepercentthreshold))
  (bind _returnvalue_15 (* _returnvalue_14 1e+02))
  (bind _returnvalue_16 (Math|Float|Floor _returnvalue_15))
  ;[1 1 1] EVENT -> On Damage Threshold (Drops loot and quest items)
  (Utilities|FlowControl|Sequence
    (:then_0
      (if (< _returnvalue_9 (Variables|LootDropOnDamage|GetLootDropCurentThreshold))
        (CallFunction|EventDropLootonDamageThreshold DamagingActor HitResult))
      (Variables|LootDropOnDamage|SetLootDropCurentThreshold _returnvalue_9) ; Update the current threshold so that it can be used for the next check
    )
    (:then_1
      (if (< _returnvalue_16 (Variables|Quests|GetRemainingQuestEventThresholds))
        (CallFunction|EventSentQuestEventonDamageThreshold DamagingPlayerController HitResult))
      (Variables|Quests|SetRemainingQuestEventThresholds _returnvalue_16))))

(event Custom|DriverCoverDeactivated
  ;[0.3591 0.6719 0.337] EVENT -> Driver Cover Activate/Deactivate
  (Variables|DriverCover|SetDriverCoverOnCooldown true)
  (bind _returnvalue (Utilities|Time|SetTimerbyEvent (delegate Custom|UpdateDriverCoverEmissive) 0.25 true))
  (Variables|DriverCover|SetDriverCoverResetTimer _returnvalue)
  (bind _returnvalue_2 (Utilities|Time|SetTimerbyEvent (delegate Custom|DriverCoverActivated) (Variables|DriverCover|GetDriverCoverResetCooldownTime)))
  (Variables|DriverCover|SetDriverCoverResetCooldownTimer _returnvalue_2)
  (Animation|Montage|PlayMontage (Variables|Vehicle|GetSkeletalMesh) (Variables|Animation|GetMontage_DriverCoverDestroyed))
  (Class|BPTankAudioController|DriverCoverDeactivateAudio (Utilities|GetAudioController) false)
  (Effects|SpawnEmitterAttached (Variables|DriverCover|GetFX_Cascade_DriverCoverDestroyed) (Variables|Vehicle|GetSkeletalMesh) "DriverCoverSocket_FX")
  (Default|CallDriverCoverEnabledChangedBroadcast false))

(event Custom|UpdateDriverCoverEmissive
  ;[0.3591 0.6719 0.337] EVENT -> Driver Cover Activate/Deactivate
  (if (not (Networking|IsServer))
    (Variables|DriverCover|SetDriverHatchCurrentHealthPercent (+ (Variables|DriverCover|GetDriverHatchCurrentHealthPercent) (Math|Float|SafeDivide 0.25 (Variables|DriverCover|GetDriverCoverResetCooldownTime))))))

(event Custom|DriverCoverActivated
  ;[0.3591 0.6719 0.337] EVENT -> Driver Cover Activate/Deactivate
  (Utilities|Time|ClearandInvalidateTimerbyHandle (Variables|DriverCover|GetDriverCoverResetTimer))
  (Variables|DriverCover|SetDriverHatchCurrentHealthPercent 1.0)
  (Variables|DriverCover|SetDriverCoverOnCooldown false)
  (Utilities|IsValid (Variables|VehicleDefaults|GetDriverPawn)
    (:Is%20Valid
      (Animation|Montage|PlayMontage (Variables|Vehicle|GetSkeletalMesh) (Variables|Animation|GetMontage_DriverCoverReactivateWithDriver))
      (Class|BPTankAudioController|DriverCoverActivateAudio (Utilities|GetAudioController) true)
      (CallFunction|RepairDamageablePart (Variables|DriverCover|GetDriverCoverDamageablePartName) (Variables|DriverCover|GetDriverHatchCurrentHealthPercent))
      (Default|CallDriverCoverEnabledChangedBroadcast true))
    (:Is%20Not%20Valid
      (Animation|Montage|PlayMontage (Variables|Vehicle|GetSkeletalMesh) (Variables|Animation|GetMontage_DriverCoverReactivateNoDriver))
      (Class|BPTankAudioController|DriverCoverActivateAudio (Utilities|GetAudioController) false))) ; If we come off cooldown when there's no driver don't repair Cover
)

(event Custom|DriverCoverDeactivateOnDriverExit
  ;[0.3591 0.6719 0.337] EVENT -> Driver Cover Activate/Deactivate
  (Class|BPTankAudioController|DriverCoverDeactivateAudio (Utilities|GetAudioController) true)
  (Animation|Montage|PlayMontage (Variables|Vehicle|GetSkeletalMesh) (Variables|Animation|GetMontage_DriverCoverDeactivateOnDriverExit)))

(event Custom|DriverCoverActivateOnDriverEnter
  ;[0.3591 0.6719 0.337] EVENT -> Driver Cover Activate/Deactivate
  (if (Variables|DriverCover|GetDriverCoverActiveOnDriverExit)
    (Variables|DriverCover|SetDriverCoverActiveOnDriverExit false) ; Reset Driver Cover Active On Exit
    (elif (Variables|DriverCover|GetDriverCoverOnCooldown)
      _))
  (Animation|Montage|PlayMontage (Variables|Vehicle|GetSkeletalMesh) (Variables|Animation|GetMontage_DriverCoverActivateOnDriverEnter))
  (CallFunction|RepairDamageablePart (Variables|DriverCover|GetDriverCoverDamageablePartName) (Variables|DriverCover|GetDriverHatchCurrentHealthPercent))
  (Default|CallDriverCoverEnabledChangedBroadcast true))

(event Custom|CannonFired
  (params (ProjectileRef :type Fort%20Projectile%20Base%20Object%20Reference))
  (bind (_returnvalue_location _returnvalue_rotation) (Math|Transform|BreakTransform (Transformation|GetSocketTransform (Variables|Vehicle|GetSkeletalMesh) "Muzzle" "RTS_World")))
  ;[1 1 1] EVENT -> Cannon Fired - Rock the Tank when the Main Cannon is fired
  (CallFunction|CannonFiredReticle)
  (if (Variables|VehicleDefaults|GetShouldRockTankOnCannonFire)
    (Utilities|IsValid ProjectileRef
      (:Is%20Valid
        (Audio|BroadcastAbilitySpinThrustatLocation 0 self (CallFunction|CalculateSpinThrustfromCannonFire (Variables|VehicleDefaults|GetTankSpinThrustOnCannonFire) ProjectileRef) 777.0) ; Rotate Vehicle in direction opposite to direction the cannon was fired
        (if (GameplayTags|HasMatchingGameplayTag "(TagName=\"Gameplay.Status.InsideTornado\")")
          (Audio|BroadcastAbilityImpactatLocation 0 self (* (Math|Rotator|GetRotationXVector _returnvalue_rotation) "1000.000000,1000.000000,1.000000") 777.0))))))

(event OnComponentBeginOverlap(UnderVehicleCollision)
  (params (OverlappedComponent :type Primitive%20Component%20Object%20Reference) (OtherActor :type Actor%20Object%20Reference) (OtherComp :type Primitive%20Component%20Object%20Reference) (OtherBodyIndex :type Integer) (bFromSweep :type Boolean) (SweepResult :type Hit%20Result%20Structure%20%28by%20ref%29))
  (bind _driverpawn (Variables|VehicleDefaults|GetDriverPawn))
  (bind _returnvalue (Utilities|IsValid _driverpawn :!class /Script/Engine.KismetSystemLibrary))
  (bind _self self)
  (bind _returnvalue_2 (select _returnvalue _driverpawn _self))
  (bind _returnvalue_3 (Ability|GetFortAbilitySystemComponent _returnvalue_2))
  ;[1 1 1] EVENT -> OnComponentOverlap: If we drive over another vehicle, destroy all their damageable parts.
  (if (Networking|IsServer)
    (if (!= OtherActor self)
      (bind _asfort_athena_vehicle (Utilities|Casting|CastToFortAthenaVehicle OtherActor))
      (if (CallFunction|DestroyVehiclePartswhenDrivenOver)
        (if (not (.bSuccess (Utilities|Casting|CastToFortTankVehicle _asfort_athena_vehicle)))
          (CallFunction|DestroyAllDamageableParts _asfort_athena_vehicle) ; Destroy damageable parts on the vehicle we drive over
          (bind _returnvalue_4 (Ability|GameplayEffect|AssignTagSetbyCallerMagnitude (GameplayEffects|MakeOutgoingSpec :self _returnvalue_3 :GameplayEffectClass (Variables|VehicleImpact|GetGE_DamageVehicleOnDriveOver) :Level 0.0) "(TagName=\"SetByCaller.Damage\")" (Variables|VehicleDefaults|GetDriveOverVehicleDamage)))
          (GameplayEffects|FortApplyGameplayEffectSpecToTarget :self _returnvalue_3 :GameplayEffectSpecHandle _returnvalue_4 :TargetOfEffect (Ability|GetAbilitySystemComponent _asfort_athena_vehicle) :OptionalSourceObject _returnvalue_2 :OptionalEffectCauser _returnvalue_2 :OptionalDamageSource _returnvalue_2) ; Apply damage to vehicle
        )))))

(event Custom|EnableGunnerHatchCollision
  (params (Enabled :type Boolean))
  (bind _returnvalue (Math|Float|SelectFloat 1.0 0.0 Enabled))
  ;[1 1 1] Enable / Disable Gunner Hatch Collision
  (for _array_element (Variables|/Script/FortniteGame.FortVehicleConfigs|GetDamageableParts (Class|FortVehicleInterface|GetVehicleConfigs))
    (switch name (Utilities|Struct|BreakVehicleDamageablePartConfig _array_element)
      (:TurretHatch_L
        (CallFunction|RepairDamageablePart (Utilities|Struct|BreakVehicleDamageablePartConfig _array_element) _returnvalue))
      (:TurretHatch_R
        (CallFunction|RepairDamageablePart (Utilities|Struct|BreakVehicleDamageablePartConfig _array_element) _returnvalue)))))

(event Custom|HandleOnExplode
  (params (LastDamageInstigator :type Controller%20Object%20Reference) (Vehicle :type Actor%20Object%20Reference))
  (bind _explodeinstigator (Variables|VehicleExplode|GetExplodeInstigator))
  (bind _returnvalue (Utilities|IsValid _explodeinstigator :!class /Script/Engine.KismetSystemLibrary))
  (bind _self self)
  (bind _returnvalue_2 (select _returnvalue _explodeinstigator _self))
  (bind _returnvalue_3 (Ability|GetFortAbilitySystemComponent _returnvalue_2))
  (bind _self_2 self)
  ;[0 2.136 2.2 1 35] EVENT -> Explode:     BOOOOOOM
  ;;[1 1 1] EVENT -> On Explode
  (CallFunction|ClearandInvalidateAllTimers) ; Any timers running should be stopped
  ;;[1 1 1] Trigger Explosion Cosmetics
  (if (Networking|IsDedicatedServer)
    (CallFunction|DestroyAllDamageableParts)
    (else
      (CallFunction|ExplodeCosmetics 0)
      (Utilities|SpawnFNRadialForce (Transformation|GetWorldTransform (Variables|Vehicle|GetSkeletalMesh)) 0.3 0.874 2.4e+03 1.75
        (:Out
          _))))
  ;;[1 1 1] Launch ourself into the air
  (Audio|BroadcastAbilityImpactatLocation 0 _self_2 (Math|Vector|MakeVector 0.0 0.0 (Variables|VehicleExplode|GetImpulseSelfOnExplodeZMagnitude)) 777.0 "None" true)
  (Audio|BroadcastAbilitySpinThrustatLocation 0 _self_2 (* (Transformation|GetRightVector (Variables|Vehicle|GetSkeletalMesh)) -5e+02) 777.0)
  ;;[1 1 1] Grab all nearby actors
  (if (Variables|VehicleExplode|GetCanExplodeHack)
    (Utilities|FlowControl|DoOnce false
      (:Completed
        (if (Networking|IsServer)
          (bind _outactors (Collision|SphereOverlapActors (Transformation|GetActorLocation) (Variables|VehicleExplode|GetExplodeRadius) (Utilities|Array|MakeArray "ObjectTypeQuery3" "ObjectTypeQuery1" "ObjectTypeQuery2" "ObjectTypeQuery5" "ObjectTypeQuery7") 0 (Utilities|Array|MakeArray self)))
          (for _array_element _outactors
            (bind _in_line_of_sight (Class|MissionBlueprintFunctionLibrary|CheckBasicActorLineOfSight self "0, 0,20.000000" _array_element "0, 0, 0" true "TraceTypeQuery1" "None" 15.0 true))
            (if _in_line_of_sight
              (Utilities|Array|AddUnique (Variables|VehicleExplode|GetTargetedActors) _array_element)))
          ;;[1 1 1] Also do damage to players in vehicle (FORT-298391)
          (for _array_element_2 (Vehicle|GetAllPassengers)
            (Utilities|Array|AddUnique (Variables|VehicleExplode|GetTargetedActors) _array_element_2))
          ;;[1 1 1] Apply Damage to nearby actors after Exploding
          (for _array_element_3 (Variables|VehicleExplode|GetTargetedActors)
            (if (Variables|Actor|GetbCanBeDamaged _array_element_3)
              (bind _returnvalue_4 (Ability|GameplayEffect|AssignTagSetbyCallerMagnitude (GameplayEffects|MakeOutgoingSpec :self _returnvalue_3 :GameplayEffectClass (select (.bSuccess (Utilities|Casting|CastToFortPlayerPawn _array_element_3)) (Variables|DamageGameplayEffects|GetGE_ExplodePlayerDamage) (select (.bSuccess (Utilities|Casting|CastToBuildingSMActor _array_element_3)) (Variables|DamageGameplayEffects|GetGE_ExplodeBuildingDamage) (Variables|DamageGameplayEffects|GetGE_ExplodeVehicleDamage))) :Level 0.0) "(TagName=\"SetByCaller.Damage\")" (select (.bSuccess (Utilities|Casting|CastToFortPlayerPawn _array_element_3)) (Variables|VehicleExplode|GetCachedExplosionDamagePlayers) (Variables|VehicleExplode|GetCachedExplosionDamageVehicles))))
              (GameplayEffects|FortApplyGameplayEffectSpecToTarget :self _returnvalue_3 :GameplayEffectSpecHandle _returnvalue_4 :TargetOfEffect (Ability|GetFortAbilitySystemComponent _array_element_3) :OptionalSourceObject self :OptionalEffectCauser self :OptionalDamageSource self)
              (Utilities|FlowControl|Sequence
                (:then_0
                  ;;[1 1 1] Knock Vehicles back
                  (bind _asfort_athena_vehicle (Utilities|Casting|CastToFortAthenaVehicle _array_element_3))
                  (Audio|BroadcastAbilityImpactatLocation 0 _asfort_athena_vehicle (CallFunction|CalcImpulseDirection _asfort_athena_vehicle (Variables|VehicleExplode|GetExplodeVehicleImpulse)) 777.0)
                  (bind _output_get (Variables|VehicleExplode|SetExplodeVehicleSpin (Math|Vector|MakeVector 0.0 (Math|Random|RandomFloatinRange (Math|Float|Min(Float) (* (Variables|VehicleExplode|GetExplodeSpinThrustMagnitude) -1.0) -5e+01) (Math|Float|Max(Float) (Variables|VehicleExplode|GetExplodeSpinThrustMagnitude) 5e+01)) (Math|Random|RandomFloatinRange (Math|Float|Min(Float) (* (Variables|VehicleExplode|GetExplodeSpinThrustMagnitude) -1.0) -5e+01) (Math|Float|Max(Float) (Variables|VehicleExplode|GetExplodeSpinThrustMagnitude) 5e+01)))))
                  (Audio|BroadcastAbilitySpinThrustatLocation 0 _asfort_athena_vehicle _output_get 777.0))
                (:then_1
                  ;;[1 1 1] Launch Player s
                  (if (.bSuccess (Utilities|Casting|CastToFortPlayerPawn _array_element_3))
                    (Pawn|Character|LaunchCharacterJump (Utilities|Casting|CastToFortPlayerPawn _array_element_3) (CallFunction|CalcImpulseDirection (Utilities|Casting|CastToFortPlayerPawn _array_element_3) (Variables|VehicleExplode|GetExplodePlayerImpulse)) false false false true))))))
          (Utilities|IsValid (Variables|VehicleExplode|GetExplodeInstigator)
            (:Is%20Valid
              (CallFunction|ApplyBuildingExplosionDamage))))))))

(event Custom|ExplodeCosmetics
  (params (DestroyedActor :type Actor%20Object%20Reference))
  (bind _returnvalue (CallFunction|GetMaxZeroHealthTime))
  (bind _prefadecleanupdelay (Variables|VehicleExplode|GetPreFadeCleanupDelay))
  (bind _returnvalue_2 (- _returnvalue _prefadecleanupdelay))
  (bind _skeletalmesh (Variables|Vehicle|GetSkeletalMesh))
  (bind _niagaracomp_enginedamage (Variables|Default|GetNiagaraComp_EngineDamage))
  ;[0 2.136 2.2 1 35] EVENT -> Explode:     BOOOOOOM
  ;;[0 0 0] Explode Cosmetics
  (Niagara|SpawnSystematLocation :SystemTemplate (Variables|FX|GetNiagaraFX_Explode) :Location (Transformation|GetActorLocation) :Scale "1.000000,1.000000,1.000000" :bAutoDestroy true :bAutoActivate true :PoolingMethod "AutoRelease" :bPreCullCheck true)
  (Utilities|IsValid _niagaracomp_enginedamage
    (:Is%20Valid
      (Components|DestroyComponent _niagaracomp_enginedamage))
    (:Is%20Not%20Valid
      _))
  (for (_ _array_index) (Rendering|Material|GetMaterials _skeletalmesh)
    (Rendering|Material|SetMaterial _skeletalmesh _array_index "/Valet/Materials/M_FN_Valet_Destroyed_Master_InstAdjusted.M_FN_Valet_Destroyed_Master_InstAdjusted"))
  (Utilities|FlowControl|Sequence
    (:then_0
      (CallFunction|SpawnWreckFX)
      (Camera|PlayWorldCameraShake "/Game/Effects/Fort_Effects/BluePrints/CameraShakes/B_Large_Explosion_CameraShake.B_Large_Explosion_CameraShake_C" (Transformation|GetActorLocation) 1.0 (Variables|CameraShakes|GetExplosion_CameraShake_Radius) 0.2 true)
      (Class|BPTankAudioController|OnDestroyAudio (Utilities|GetAudioController)))
    (:then_2
      (if (> _returnvalue_2 0)
        (Utilities|FlowControl|Delay _returnvalue_2) ; Wait to turn off the FX until after we dissolve
      )
      (Variables|VehicleExplode|SetVehicleExploded true) ; Don't update FX anymore
      (CallFunction|CleanupLoopingPooledFXNativize))))

(event Custom|SpawnWreckFX
  ;[0 2.136 2.2 1 35] EVENT -> Explode:     BOOOOOOM
  ;;[0 0 0] Spawn wreck Niagara fx system
  (bind _returnvalue (Niagara|SpawnSystemAttached (Variables|FX|GetNiagaraFX_Wreck) (Variables|Vehicle|GetSkeletalMesh) "Driver" "0, 0, 0" "0, 0, 0" "SnapToTarget" false true "AutoRelease"))
  (bind _output_get (Variables|Default|SetNS_WreckFX _returnvalue))
  (Utilities|IsValid _output_get
    (:Is%20Valid
      (Niagara|SetNiagaraVariable(Float) _output_get "Duration" (CallFunction|GetMaxZeroHealthTime)))))

(event Custom|HandleOnStartCriticalHealth
  (params (LastDamageInstigator :type Controller%20Object%20Reference))
  ;[1 1 1] EVENT -> Start Critical Health (Curie Fire)
  (Variables|VehicleExplode|SetVehicleExploded true) ; Don't update FX anymore
  (Utilities|FlowControl|DoOnce false
    (:Completed
      (if (Variables|VehicleDefaults|GetCurieEnabled)
        (Curie|OnCurieActivated
          (:OnCurieActive
            (Curie|EnableElementAttachmentonActor self "(TagName=\"Curie.Element.Fire\")")
            (Curie|ExecuteFireApplication self self self 1e+02 (Collision|MakeHitResult false false 0.0 0.0 (Transformation|GetActorLocation) (Transformation|GetActorLocation) "0,0,1" "0,0,1" 0 self (Variables|Vehicle|GetSkeletalMesh)) -1 128.0)))))))

(event OnEnteredWaterDelegate(PontoonsComponent)
  (params (Pontoon :type Spherical%20Pontoon%20Structure%20%28by%20ref%29))
  ;[0.6084 0.8133 1 1 33] EVENT -> Enter/Exit Water Body
  (CallFunction|EnterWaterBodyServer))

(event EventBeginPlay
  ;[1 1 1] EVENT -> Begin Play
  (Utilities|FlowControl|Sequence
    (:then_0
      (CallFunction|LocalPassengerMatSwap)
      (CallFunction|DamgeStateSetAmount)
      (Vehicle|Fuel|BindEventtoOnFuelChanged (Variables|Default|GetFuelComponent) (delegate Custom|EventOnFuelChanged))
      (EventDispatchers|BindEventtoOnRefueledfromEmptyDelegate (EventDispatchers|CreateEvent :!event HandleRefueledFromEmpty))
      (Collision|AssignOnDamageablePartHealthChanged (delegate Custom|OnDamageablePartHealthChanged_Event))
      (Variables|VehicleImpact|SetIgnoreHitLastDriver true)
      (Variables|VehicleImpact|SetPassengerIsInstigator true)
      (CallFunction|SetPropDestructionParams (Utilities|Array|MakeArray "/Script/FortniteGame.BuildingProp" "/Game/Building/ActorBlueprints/Containers/ParentBuildingContainerBlueprint.ParentBuildingContainerBlueprint_C") (Utilities|Array|MakeArray) (Utilities|Array|MakeArray "/Game/Building/EditModePatterns/Wall/EMP_Wall_HalfWallS.EMP_Wall_HalfWallS") (Utilities|Array|MakeArray))
      (Class|BPTankAudioController|CacheReferenceVehicle (Utilities|GetAudioController) self)
      (CallFunction|CacheExplosionDamageValues)
      (Health|BindEventtoOnHealthChanged (EventDispatchers|CreateEvent :!event Custom|HealthChanged) :!class /Script/FortniteGame.FortAthenaVehicle)
      ;; If server, nuke audio and fx
      (if (Networking|IsDedicatedServer)
        _)
      (Vehicle|Destruction|BindEventtoOnVehicleExplodeDelegate (EventDispatchers|CreateEvent :!event Custom|HandleOnExplode) :!class /Script/FortniteGame.FortAthenaVehicle)
      (Vehicle|Destruction|BindEventtoOnStartCriticalHealth (EventDispatchers|CreateEvent :!event Custom|HandleOnStartCriticalHealth)))
    (:then_1
      (Class|FortAsyncActionGameStateReady|GameStateReadyAsync
        (:OnReady
          (if (and (Spawn|GetForcedtoAlwaysSpawn) (Networking|IsServer))
            (Class|FuelComponent|SetFuel (Variables|Default|GetFuelComponent) 5e+02)))))
    (:then_2
      ;;[1 1 1] Bind Pawn Enter and Exit Vehicle Seats.
      (Vehicle|AssignOnPawnExitedVehicleSeat (EventDispatchers|CreateEvent :!event Custom|PawnExitVehicleSeat) :!class /Script/FortniteGame.FortAthenaVehicle)
      (Vehicle|AssignOnPawnEnteredVehicleSeat (EventDispatchers|CreateEvent :!event Custom|PawnEnteredVehicleSeat) :!class /Script/FortniteGame.FortAthenaVehicle)
      (Tank|Interact|BindEventtoOnTankRepairedDelegate (EventDispatchers|CreateEvent :!event Custom|TankTreadRepaired)))))

(event Custom|EnterWaterBodyServer
  ;[0.6084 0.8133 1 1 33] EVENT -> Enter/Exit Water Body
  (Variables|Default|SetNumPontoonInWater (Math|Integer|Clamp(Integer) (+ (Variables|Default|GetNumPontoonInWater) 1) 0 6)) ; Keep Track of the number of pontoons in water so we can trigger the Exit Event only when there are no pontoons left in the water
  (CallFunction|EnterWaterBody))

(event EventEndPlay
  (params (EndPlayReason :type EEndPlayReason%20Enum :default "Destroyed"))
  (bind _audiocontroller (Utilities|GetAudioController))
  ;[1 0.4122 0] EVENT -> End Play
  (|CallParentFunction EndPlayReason)
  (CallFunction|DestroyRadio)
  (Utilities|FlowControl|Sequence
    (:then_0
      (Utilities|IsValid _audiocontroller
        (:Is%20Valid
          (Class|BPTankAudioController|VehicleEndPlayAudio _audiocontroller))))
    (:then_1
      (CallFunction|CleanupLoopingPooledFXNativize) ; Try and insure clean-up of lingering FX: FORT-559707
    )))

(event Custom|OnDamageablePartHealthChanged_Event
  (params (PartConfig :type Vehicle%20Damageable%20Part%20Config%20Structure%20%28by%20ref%29) (Part :type Vehicle%20Damageable%20Part%20Structure%20%28by%20ref%29) (bNewlyDisabled :type Boolean) (bNewlyEnabled :type Boolean))
  ;[1 1 1] EVENT -> Begin Play
  (CallFunction|DamageablePartChanged PartConfig Part bNewlyDisabled))

(event Custom|EventOnFuelChanged
  (params (NewFuelPercent :type Float%20%28single-precision%29) (OldFuelPercent :type Float%20%28single-precision%29))
  ;[1 1 1] EVENT -> Begin Play
  (Variables|VehicleDefaults|SetCurrentFuelPercent NewFuelPercent))

(event Custom|EnterWaterBody
  (bind _watersurfacedatacache (Variables|FortAthenaVehicle|GetWaterSurfaceDataCache))
  (bind (_waterplanelocation _waterplanenormal _watersurfacelocation _waterdepth _waterbodyidx) (Utilities|Struct|BreakWaterSurfaceInfo _watersurfacedatacache))
  (bind (_returnvalue_x _returnvalue_y _returnvalue_z) (Math|Vector|BreakVector (Transformation|GetActorLocation)))
  (bind (_returnvalue_x_2 _returnvalue_y_2 _returnvalue_z_2) (Math|Vector|BreakVector (Transformation|GetActorLocation)))
  (bind _niagaracomp_inwater (Variables|FX|GetNiagaraComp_InWater))
  ;[0.6084 0.8133 1 1 33] EVENT -> Enter/Exit Water Body
  (Utilities|FlowControl|DoOnce false
    (:Completed
      (Class|BPTankAudioController|OnWaterStateChangeAudio (Utilities|GetAudioController) true)
      (bind _ns_tank_roadandexhaust (Variables|Default|GetNS_Tank_RoadAndExhaust)
        (:then
          (Niagara|SetNiagaraVariable(LinearColor) _ns_tank_roadandexhaust "DustTireBlurColor" (0.3 0.6 1.0 1.0)))
        (:else
          _))
      (Niagara|SpawnSystemAttached (Variables|FX|GetNiagaraFX_EnterWaterSplash) (Variables|Vehicle|GetSkeletalMesh) "WaterTest" "0, 0, 0" "0, 0, 0" "KeepRelativeOffset" true) ; Splash on EnterWater
      (bind _returnvalue (Utilities|Time|SetTimerbyEvent (delegate Custom|SnapFoamtoWaterSurface) 0.2 true))
      (Variables|WaterInteraction|SetTimerSnapFoamToWaterSurface _returnvalue)
      (bind _niagaracomp_inwater_2 (Variables|FX|GetNiagaraComp_InWater)
        (:then
          (Components|Activation|Activate _niagaracomp_inwater_2 false))
        (:else
          (bind _returnvalue_2 (Niagara|SpawnSystemAttached (Variables|FX|GetNiagaraFX_InWater) (Variables|Vehicle|GetSkeletalMesh) "WaterTest" "0, 0, 0" "0, 0, 0" "KeepRelativeOffset" true))
          (Variables|FX|SetNiagaraComp_InWater _returnvalue_2)))
      (if (not (Networking|IsServer))
        (Utilities|FlowControl|Sequence
          (:then_0
            ;;[1 1 1] Trigger large splash base on falling speed
            (Niagara|SetNiagaraVariable(Float) _niagaracomp_inwater "LoopFoamSpawn" 0.0)
            (Variables|WaterInteraction|SetVehicleFallSpeed _returnvalue_z)
            (Utilities|FlowControl|Delay 0.08)
            (if (> (Math|Float|Absolute(Float) (- (Variables|WaterInteraction|GetVehicleFallSpeed) _returnvalue_z_2)) 1e+01)
              (Niagara|SetNiagaraVariable(Int32) _niagaracomp_inwater "SplashSpawn" 1)
              (Niagara|SetNiagaraVariable(Float) _niagaracomp_inwater "BubbleSpawn" 0.0)
              (Niagara|SetNiagaraVariable(Int32) _niagaracomp_inwater "BigBubbleSpawn" 0)
              (Niagara|SetNiagaraVariable(Vector3) _niagaracomp_inwater "WaterPlaneLocation" (CallFunction|WaterSurfaceVector))
              (Niagara|SetNiagaraVariable(Float) _niagaracomp_inwater "WaterDepth" _waterdepth)
              (Niagara|SetNiagaraVariable(Vector3) _niagaracomp_inwater "WaterPlaneNormal" _waterplanenormal)
              (Niagara|SetNiagaraVariable(Float) _niagaracomp_inwater "WaterBodyIndex" (Math|Conversions|ToFloat(Integer) _waterbodyidx))
              (else
                (Niagara|SetNiagaraVariable(Int32) _niagaracomp_inwater "SplashSpawn" 0)
                (Niagara|SetNiagaraVariable(Float) _niagaracomp_inwater "BubbleSpawn" 0.0)
                (Niagara|SetNiagaraVariable(Int32) _niagaracomp_inwater "BigBubbleSpawn" 0))))
          (:then_1
            ;;[1 1 1] Deactivate Water bubble FX while sinking to certain point and no driver in vehicle
            (Utilities|FlowControl|Delay 12.0)
            (Variables|VehicleDefaults|GetDriverPawn
              (:then)
              (:else
                (Components|Activation|Deactivate _niagaracomp_inwater)
                (Utilities|Time|ClearandInvalidateTimerbyHandle (Variables|WaterInteraction|GetTimerSnapFoamToWaterSurface))
                (Development|PrintString "DeactiveWaterFXwhenDriverIsNotValid")))))))))

(event Custom|CE_PawnEnterVehicleWhileInWater
  (bind _watersurfacedatacache (Variables|FortAthenaVehicle|GetWaterSurfaceDataCache))
  (bind (_waterplanelocation _waterplanenormal _watersurfacelocation _waterdepth _waterbodyidx) (Utilities|Struct|BreakWaterSurfaceInfo _watersurfacedatacache))
  (bind (_returnvalue_x _returnvalue_y _returnvalue_z) (Math|Vector|BreakVector (Transformation|GetActorLocation)))
  (bind (_returnvalue_x_2 _returnvalue_y_2 _returnvalue_z_2) (Math|Vector|BreakVector (Transformation|GetActorLocation)))
  (bind _niagaracomp_inwater (Variables|FX|GetNiagaraComp_InWater))
  ;[0.6084 0.8133 1 1 33] EVENT -> Enter/Exit Water Body
  (if (> (Variables|Default|GetNumPontoonInWater) 0)
    (bind _returnvalue (Utilities|Time|SetTimerbyEvent (delegate Custom|SnapFoamtoWaterSurface) 0.2 true))
    (Variables|WaterInteraction|SetTimerSnapFoamToWaterSurface _returnvalue)
    (bind _niagaracomp_inwater_2 (Variables|FX|GetNiagaraComp_InWater)
      (:then
        (Components|Activation|Activate _niagaracomp_inwater_2 false))
      (:else
        (bind _returnvalue_2 (Niagara|SpawnSystemAttached (Variables|FX|GetNiagaraFX_InWater) (Variables|Vehicle|GetSkeletalMesh) "WaterTest" "0, 0, 0" "0, 0, 0" "KeepRelativeOffset" true))
        (Variables|FX|SetNiagaraComp_InWater _returnvalue_2)))
    (if (not (Networking|IsServer))
      (Utilities|FlowControl|Sequence
        (:then_0
          ;;[1 1 1] Trigger large splash base on falling speed
          (Niagara|SetNiagaraVariable(Float) _niagaracomp_inwater "LoopFoamSpawn" 0.0)
          (Variables|WaterInteraction|SetVehicleFallSpeed _returnvalue_z)
          (Utilities|FlowControl|Delay 0.08)
          (if (> (Math|Float|Absolute(Float) (- (Variables|WaterInteraction|GetVehicleFallSpeed) _returnvalue_z_2)) 1e+01)
            (Niagara|SetNiagaraVariable(Int32) _niagaracomp_inwater "SplashSpawn" 1)
            (Niagara|SetNiagaraVariable(Float) _niagaracomp_inwater "BubbleSpawn" 0.0)
            (Niagara|SetNiagaraVariable(Int32) _niagaracomp_inwater "BigBubbleSpawn" 0)
            (Niagara|SetNiagaraVariable(Vector3) _niagaracomp_inwater "WaterPlaneLocation" (CallFunction|WaterSurfaceVector))
            (Niagara|SetNiagaraVariable(Float) _niagaracomp_inwater "WaterDepth" _waterdepth)
            (Niagara|SetNiagaraVariable(Vector3) _niagaracomp_inwater "WaterPlaneNormal" _waterplanenormal)
            (Niagara|SetNiagaraVariable(Float) _niagaracomp_inwater "WaterBodyIndex" (Math|Conversions|ToFloat(Integer) _waterbodyidx))
            (else
              (Niagara|SetNiagaraVariable(Int32) _niagaracomp_inwater "SplashSpawn" 0)
              (Niagara|SetNiagaraVariable(Float) _niagaracomp_inwater "BubbleSpawn" 0.0)
              (Niagara|SetNiagaraVariable(Int32) _niagaracomp_inwater "BigBubbleSpawn" 0))))
        (:then_1
          ;;[1 1 1] Deactivate Water bubble FX while sinking to certain point and no driver in vehicle
          (Utilities|FlowControl|Delay 12.0)
          (Variables|VehicleDefaults|GetDriverPawn
            (:then)
            (:else
              (Components|Activation|Deactivate _niagaracomp_inwater)
              (Utilities|Time|ClearandInvalidateTimerbyHandle (Variables|WaterInteraction|GetTimerSnapFoamToWaterSurface))
              (Development|PrintString "DeactiveWaterFXwhenDriverIsNotValid"))))))))

(event Custom|ResetWaterDoOnce
  (bind _watersurfacedatacache (Variables|FortAthenaVehicle|GetWaterSurfaceDataCache))
  (bind (_waterplanelocation _waterplanenormal _watersurfacelocation _waterdepth _waterbodyidx) (Utilities|Struct|BreakWaterSurfaceInfo _watersurfacedatacache))
  (bind (_returnvalue_x _returnvalue_y _returnvalue_z) (Math|Vector|BreakVector (Transformation|GetActorLocation)))
  (bind (_returnvalue_x_2 _returnvalue_y_2 _returnvalue_z_2) (Math|Vector|BreakVector (Transformation|GetActorLocation)))
  (bind _niagaracomp_inwater (Variables|FX|GetNiagaraComp_InWater))
  ;[0.6084 0.8133 1 1 33] EVENT -> Enter/Exit Water Body
  (Utilities|FlowControl|RetriggerableDelay 1.5) ; Delay because we don't want to fire off if we are bouncing in and out of a water body quickly
  (Utilities|FlowControl|DoOnce false
    (:Completed
      (Class|BPTankAudioController|OnWaterStateChangeAudio (Utilities|GetAudioController) true)
      (bind _ns_tank_roadandexhaust (Variables|Default|GetNS_Tank_RoadAndExhaust)
        (:then
          (Niagara|SetNiagaraVariable(LinearColor) _ns_tank_roadandexhaust "DustTireBlurColor" (0.3 0.6 1.0 1.0)))
        (:else
          _))
      (Niagara|SpawnSystemAttached (Variables|FX|GetNiagaraFX_EnterWaterSplash) (Variables|Vehicle|GetSkeletalMesh) "WaterTest" "0, 0, 0" "0, 0, 0" "KeepRelativeOffset" true) ; Splash on EnterWater
      (bind _returnvalue (Utilities|Time|SetTimerbyEvent (delegate Custom|SnapFoamtoWaterSurface) 0.2 true))
      (Variables|WaterInteraction|SetTimerSnapFoamToWaterSurface _returnvalue)
      (bind _niagaracomp_inwater_2 (Variables|FX|GetNiagaraComp_InWater)
        (:then
          (Components|Activation|Activate _niagaracomp_inwater_2 false))
        (:else
          (bind _returnvalue_2 (Niagara|SpawnSystemAttached (Variables|FX|GetNiagaraFX_InWater) (Variables|Vehicle|GetSkeletalMesh) "WaterTest" "0, 0, 0" "0, 0, 0" "KeepRelativeOffset" true))
          (Variables|FX|SetNiagaraComp_InWater _returnvalue_2)))
      (if (not (Networking|IsServer))
        (Utilities|FlowControl|Sequence
          (:then_0
            ;;[1 1 1] Trigger large splash base on falling speed
            (Niagara|SetNiagaraVariable(Float) _niagaracomp_inwater "LoopFoamSpawn" 0.0)
            (Variables|WaterInteraction|SetVehicleFallSpeed _returnvalue_z)
            (Utilities|FlowControl|Delay 0.08)
            (if (> (Math|Float|Absolute(Float) (- (Variables|WaterInteraction|GetVehicleFallSpeed) _returnvalue_z_2)) 1e+01)
              (Niagara|SetNiagaraVariable(Int32) _niagaracomp_inwater "SplashSpawn" 1)
              (Niagara|SetNiagaraVariable(Float) _niagaracomp_inwater "BubbleSpawn" 0.0)
              (Niagara|SetNiagaraVariable(Int32) _niagaracomp_inwater "BigBubbleSpawn" 0)
              (Niagara|SetNiagaraVariable(Vector3) _niagaracomp_inwater "WaterPlaneLocation" (CallFunction|WaterSurfaceVector))
              (Niagara|SetNiagaraVariable(Float) _niagaracomp_inwater "WaterDepth" _waterdepth)
              (Niagara|SetNiagaraVariable(Vector3) _niagaracomp_inwater "WaterPlaneNormal" _waterplanenormal)
              (Niagara|SetNiagaraVariable(Float) _niagaracomp_inwater "WaterBodyIndex" (Math|Conversions|ToFloat(Integer) _waterbodyidx))
              (else
                (Niagara|SetNiagaraVariable(Int32) _niagaracomp_inwater "SplashSpawn" 0)
                (Niagara|SetNiagaraVariable(Float) _niagaracomp_inwater "BubbleSpawn" 0.0)
                (Niagara|SetNiagaraVariable(Int32) _niagaracomp_inwater "BigBubbleSpawn" 0))))
          (:then_1
            ;;[1 1 1] Deactivate Water bubble FX while sinking to certain point and no driver in vehicle
            (Utilities|FlowControl|Delay 12.0)
            (Variables|VehicleDefaults|GetDriverPawn
              (:then)
              (:else
                (Components|Activation|Deactivate _niagaracomp_inwater)
                (Utilities|Time|ClearandInvalidateTimerbyHandle (Variables|WaterInteraction|GetTimerSnapFoamToWaterSurface))
                (Development|PrintString "DeactiveWaterFXwhenDriverIsNotValid")))))))))

(event Custom|ResetEnterWaterDoOnceServer
  ;[0.6084 0.8133 1 1 33] EVENT -> Enter/Exit Water Body
  (CallFunction|ResetWaterDoOnce))

(event Custom|StartOutOfGasRumble
  ;[1 1 1] EVENT -> Out of Fuel
  (Game|Feedback|PlayDynamicForceFeedback (Variables|VehicleDefaults|GetDriverPlayerController) (stale EventOnEmptyFuel _intensity) 1.6 true false true false))

(event EventOnEnteredWaterBody
  (params (WaterBody :type Fort%20Water%20Body%20Actor%20Object%20Reference))
  ;[0.6084 0.8133 1 1 33] EVENT -> Enter/Exit Water Body
  (Variables|WaterInteraction|SetWaterBody WaterBody))

(event Custom|SnapFoamtoWaterSurface
  (bind _watersurfacedatacache (Variables|FortAthenaVehicle|GetWaterSurfaceDataCache))
  (bind (_waterplanelocation _waterplanenormal _watersurfacelocation _waterdepth _waterbodyidx) (Utilities|Struct|BreakWaterSurfaceInfo _watersurfacedatacache))
  (bind _returnvalue (Transformation|GetActorLocation))
  (bind _z (.z _returnvalue))
  (bind _z_2 (.z _waterplanelocation))
  (bind _returnvalue_2 (- _z_2 _z))
  (bind _niagaracomp_inwater (Variables|FX|GetNiagaraComp_InWater))
  (bind _niagaracomp_inwater_2 (Variables|FX|GetNiagaraComp_InWater))
  (bind _ns_tank_roadandexhaust (Variables|Default|GetNS_Tank_RoadAndExhaust))
  (bind _niagaracomp_inwater_3 (Variables|FX|GetNiagaraComp_InWater))
  ;[0.6084 0.8133 1 1 33] EVENT -> Enter/Exit Water Body
  (Utilities|FlowControl|Sequence
    (:then_0
      (if (Math|Float|InRange(Float) _returnvalue_2 -8e+01 8e+02)
        (bind _niagaracomp_inwater_4 (Variables|FX|GetNiagaraComp_InWater)
          (:then
            (if (Math|Float|InRange(Float) _returnvalue_2 -39.900002 2.8e+02)
              (Variables|WaterInteraction|SetIsInDeepWater false)
              (Variables|VehicleDefaults|GetDriverPawn
                (:then
                  ;;[1 1 1] Shallow Water
                  (Components|Activation|Activate _niagaracomp_inwater_4 false))
                (:else
                  _))
              (Niagara|SetNiagaraVariable(Int32) _niagaracomp_inwater_4 "SplashSpawn" 0)
              (Niagara|SetNiagaraVariable(Float) _niagaracomp_inwater_4 "BubbleSpawn" 0.0)
              (Niagara|SetNiagaraVariable(Float) _niagaracomp_inwater_4 "LoopFoamSpawn" 1.0)
              (Niagara|SetNiagaraVariable(Int32) _niagaracomp_inwater_4 "BigBubbleSpawn" 0)
              (Niagara|SetNiagaraVariable(Bool) _ns_tank_roadandexhaust "IsWaterDeep" false)
              (CallFunction|CEWaterFXResetDeepWaterEffectActivate)
              ;;[1 1 1] Foam Snap Function
              (Utilities|FlowControl|Sequence
                (:then_0
                  (Niagara|SetNiagaraVariable(Vector3) (Variables|FX|GetNiagaraComp_InWater) "WaterPlaneLocation" (CallFunction|WaterSurfaceVector)))
                (:then_1
                  (Niagara|SetNiagaraVariable(Float) _niagaracomp_inwater "WaterDepth" _waterdepth)
                  (Niagara|SetNiagaraVariable(Vector3) _niagaracomp_inwater "WaterPlaneNormal" _waterplanenormal)
                  (Niagara|SetNiagaraVariable(Float) _niagaracomp_inwater "WaterBodyIndex" (Math|Conversions|ToFloat(Integer) _waterbodyidx)))
                (:then_2
                  ;;[1 1 1] Tread Spinning in water
                  (if (Math|Float|InRange(Float) _returnvalue_2 -29.0 1e+02)
                    (bind _niagaracomp_inwater_5 (Variables|FX|GetNiagaraComp_InWater)
                      (:then
                        (Niagara|SetNiagaraVariable(Float) _niagaracomp_inwater_5 "TireSpinWaterSpawn" 1.0)
                        (if (Vehicle|PlayerIsTryingtoMoveForward)
                          (Niagara|SetNiagaraVariable(Float) _niagaracomp_inwater_2 "TireSpinWaterDirection" -1.0)
                          (elif (Vehicle|PlayerIsTryingtoReverse)
                            (Niagara|SetNiagaraVariable(Float) _niagaracomp_inwater_2 "TireSpinWaterDirection" 1.0))))
                      (:else))
                    (else
                      (bind _niagaracomp_inwater_6 (Variables|FX|GetNiagaraComp_InWater)
                        (:then
                          (Niagara|SetNiagaraVariable(Float) _niagaracomp_inwater_6 "TireSpinWaterSpawn" 0.0))
                        (:else))))
                  (Niagara|SetNiagaraVariable(Float) _niagaracomp_inwater_6 "TireSpinWaterSpawn" 0.0)))
              (else
                ;;[1 1 1] Deep Water
                (Variables|WaterInteraction|SetIsInDeepWater true)
                (Utilities|FlowControl|Sequence
                  (:then_0
                    (Utilities|FlowControl|DoOnce false
                      (:Completed
                        (Variables|VehicleDefaults|GetDriverPawn
                          (:then
                            (Components|Activation|Activate _niagaracomp_inwater_3 true)
                            (Niagara|SetNiagaraVariable(Bool) _ns_tank_roadandexhaust "IsWaterDeep" true))
                          (:else)))))
                  (:then_1
                    (Niagara|SetNiagaraVariable(Int32) _niagaracomp_inwater_3 "SplashSpawn" 0)
                    (Niagara|SetNiagaraVariable(Float) _niagaracomp_inwater_3 "BubbleSpawn" 1.0)
                    (Niagara|SetNiagaraVariable(Float) _niagaracomp_inwater_3 "LoopFoamSpawn" 0.0)
                    (Niagara|SetNiagaraVariable(Int32) _niagaracomp_inwater_3 "BigBubbleSpawn" 1))))))
          (:else))
        (else
          _)))
    (:then_1
      (if (> (Variables|Default|GetNumPontoonInWater) 0)
        _
        (else
          _))))
  (Utilities|Time|ClearandInvalidateTimerbyHandle (Variables|WaterInteraction|GetTimerSnapFoamToWaterSurface))
  (Components|Activation|Deactivate _niagaracomp_inwater_3))

(event EventOnEmptyFuel
  ;[1 1 1] EVENT -> Out of Fuel
  (Variables|Fuel|SetVehicleHasFuel false)
  (CallFunction|StartOutOfGasRumble)
  (Utilities|BoostCameraBlur false
    (:%20
      (Audio|SpawnSoundAttached (Variables|Audio|GetOutOfFuelSound) (Variables|Vehicle|GetSkeletalMesh))
      (bind _drivercontroller (Utilities|IsValidDriverController)
        (:Is%20Valid
          (CameraShakes|StartCameraShake :self (Variables|/Script/Engine.PlayerController|GetPlayerCameraManager _drivercontroller) :ShakeClass (Variables|CameraShakes|GetOutOfFuelCameraShake) :Scale 1.0 :PlaySpace "CameraLocal")
          (bind (_direction _intensity) (Timeline|OutOfFuel PlayFromStart)
            (:Update
              (goto _label Update))
            (:Finished
              (Utilities|FlowControl|Sequence
                (:then_0
                  (goto _label Stop))
                (:then_1
                  (Niagara|SetNiagaraVariable(Bool) (Variables|Default|GetNS_Tank_RoadAndExhaust) "VehicleHasFuel" false)))))
          (Game|Feedback|PlayDynamicForceFeedback (Variables|VehicleDefaults|GetDriverPlayerController) _intensity 1.6 true false true false :!label _label))))) ; Turn off Boost Blur if we run out of gas.
)

(event Custom|StopOutOfGasRumble
  ;[1 1 1] EVENT -> Out of Fuel
  (Game|Feedback|PlayDynamicForceFeedback (Variables|VehicleDefaults|GetDriverPlayerController) (stale EventOnEmptyFuel _intensity) 1.6 true false true false))

(event Custom|LocalPassengerADS
  (params (AimDownSights :type Boolean))
  (bind _mid_treads (Variables|MaterialsandTextures|GetMID_Treads))
  ;[1 1 1] Set Player Is ADS bool to turn off masking
  (bind _output_get (Variables|VehicleDefaults|SetPlayerIsADS AimDownSights))
  (if _output_get
    (Rendering|Material|SetScalarParameterValue _mid_treads "PlayerIsADS" 1.0 :!class /Script/Engine.MaterialInstanceDynamic)
    (else
      (Rendering|Material|SetScalarParameterValue _mid_treads "PlayerIsADS" 0.0 :!class /Script/Engine.MaterialInstanceDynamic))))

(event Custom|LocalPassengerMatSwap
  (bind _islocalplayerpassenger (Variables|VehicleDefaults|GetIsLocalPlayerPassenger))
  ;[1 1 1] Manage Local Passenger Material Swaps
  (Utilities|FlowControl|Sequence
    (:then_0
      (Class|FortVehicleInterface|ApplyMaterialandUseWrapifPossible (select _islocalplayerpassenger (Variables|TankMaterialSwapforLocalPlayers|GetMI_Tank_Treads_Masked) (Variables|TankMaterialSwapForLocalPlayers|GetMI_Tank_Treads)) 0 (delegate Custom|OnMIDCreated_SlotZero))
      (Class|FortVehicleInterface|ApplyMaterialandUseWrapifPossible (select _islocalplayerpassenger (Variables|TankMaterialSwapforLocalPlayers|GetMI_Tank_Parts_Masked) (Variables|TankMaterialSwapforLocalPlayers|GetMI_Tank_Parts)) 1 (delegate Custom|OnMIDCreated_Slot1))
      (Class|FortVehicleInterface|ApplyMaterialandUseWrapifPossible (select _islocalplayerpassenger (Variables|TankMaterialSwapforLocalPlayers|GetMI_Tank_Turret_Masked) (Variables|TankMaterialSwapforLocalPlayers|GetMI_Tank_Turret)) 2 (delegate Custom|OnMIDCreated_Slot2))
      (Class|FortVehicleInterface|ApplyMaterialandUseWrapifPossible (select _islocalplayerpassenger (Variables|TankMaterialSwapforLocalPlayers|GetMI_Tank_Body_Masked) (Variables|TankMaterialSwapforLocalPlayers|GetMI_Tank_Body)) 3 (delegate Custom|OnMIDCreated_Slot3)))
    (:then_1
      (Class|FortVehicleInterface|ApplyMaterialandUseWrapifPossible (select _islocalplayerpassenger (Variables|TankMaterialSwapforLocalPlayers|GetMI_Tank_BodyLOD1_Masked) (Variables|TankMaterialSwapforLocalPlayers|GetMI_Tank_BodyLOD1)) 4 (delegate Custom|OnMIDCreated_Slot4))
      (Class|FortVehicleInterface|ApplyMaterialandUseWrapifPossible (select _islocalplayerpassenger (Variables|TankMaterialSwapforLocalPlayers|GetMI_Tank_PartsLOD1_Masked) (Variables|TankMaterialSwapforLocalPlayers|GetMI_Tank_PartsLOD1)) 5 (delegate Custom|OnMIDCreated_Slot5))
      (Class|FortVehicleInterface|ApplyMaterialandUseWrapifPossible (select _islocalplayerpassenger (Variables|TankMaterialSwapforLocalPlayers|GetMI_Tank_TurretLOD1_Masked) (Variables|TankMaterialSwapforLocalPlayers|GetMI_Tank_TurretLOD1)) 6 (delegate Custom|OnMIDCreated_Slot6))
      (CallFunction|HealthChanged) ; Need this or the material swap will cause the tank to look undamaged
    )))

(event Curie|EventOnCurieContainerAcquired
  (params (CurieContainerHandle :type Curie%20Container%20Handle%20Structure))
  ;[1 1 1] EVENT -> CurieContainerAcquired
  (if (not (CallFunction|IsonFire))
    (Curie|DisableElementAttachmentonActor self "(TagName=\"Curie.Element.Fire\")")))

(event EventOnWaterTooDeep
  ;[0.6084 0.8133 1 1 33] EVENT -> Enter/Exit Water Body
  ;;[1 0.6877 0.2311] Kick players + deactive dmg vfx when too deep in water
  (|CallParentFunction))

(event Custom|CE_WaterFX_ResetDeepWaterEffectActivate
  ;[0.6084 0.8133 1 1 33] EVENT -> Enter/Exit Water Body
  ;;[1 1 1] Deep Water
  (Utilities|FlowControl|DoOnce false
    (:Completed
      (Variables|VehicleDefaults|GetDriverPawn
        (:then
          (Components|Activation|Activate (Variables|FX|GetNiagaraComp_InWater) true)
          (Niagara|SetNiagaraVariable(Bool) (Variables|Default|GetNS_Tank_RoadAndExhaust) "IsWaterDeep" true))
        (:else)))))

(event Custom|OnMIDCreated_SlotZero
  (params (MID :type Material%20Instance%20Dynamic%20Object%20Reference))
  ;[1 1 1] Manage Local Passenger Material Swaps
  (Variables|MaterialsandTextures|SetMID_Treads MID))

(event Custom|OnMIDCreated_Slot1
  (params (MID :type Material%20Instance%20Dynamic%20Object%20Reference))
  ;[1 1 1] Manage Local Passenger Material Swaps
  (Variables|MaterialsandTextures|SetMID_Parts MID))

(event Custom|OnMIDCreated_Slot2
  (params (MID :type Material%20Instance%20Dynamic%20Object%20Reference))
  ;[1 1 1] Manage Local Passenger Material Swaps
  (Variables|MaterialsandTextures|SetMID_Turret MID))

(event Custom|OnMIDCreated_Slot3
  (params (MID :type Material%20Instance%20Dynamic%20Object%20Reference))
  ;[1 1 1] Manage Local Passenger Material Swaps
  (Variables|MaterialsandTextures|SetMID_Body MID))

(event Custom|CreateRadio
  ;[1 1 1] Manage Radio
  (bind _returnvalue (Game|SpawnActorfromClass "/Valet/Radio/BP_Vehicle_Radio.BP_Vehicle_Radio_C" (Transformation|GetActorTransform) "Undefined" "OverrideRootScale" 0 0))
  (Variables|Radio|SetVehicleRadio _returnvalue))

(event Vehicle|EventOnHonk
  ;[1 1 1] EVENT -> Honk
  (Class|BPTankAudioController|OnHornAudio (Utilities|GetAudioController)))

(event EventOnGearShiftUp
  (bind _ns_tank_roadandexhaust (Variables|Default|GetNS_Tank_RoadAndExhaust))
  ;[1 1 1] Exhaust Fire Blip
  (Niagara|SetNiagaraVariable(Bool) _ns_tank_roadandexhaust "Downshifting" true)
  (Niagara|SetNiagaraVariable(Bool) _ns_tank_roadandexhaust "IsShifting" true)
  (Utilities|FlowControl|Delay 0.125)
  (Niagara|SetNiagaraVariable(Bool) _ns_tank_roadandexhaust "IsShifting" false)
  (Niagara|SetNiagaraVariable(Bool) _ns_tank_roadandexhaust "Downshifting" false))

(event Custom|ExitWaterBody
  (bind _niagaracomp_inwater (Variables|FX|GetNiagaraComp_InWater))
  ;[0.6084 0.8133 1 1 33] EVENT -> Enter/Exit Water Body
  ;;[1 1 1] EVENT -> On Exited Water Body: Deactivate water FX When vehicle is out of water
  (Class|BPTankAudioController|OnWaterStateChangeAudio (Utilities|GetAudioController) false)
  (bind _ns_tank_roadandexhaust (Variables|Default|GetNS_Tank_RoadAndExhaust)
    (:then
      (Niagara|SetNiagaraVariable(LinearColor) _ns_tank_roadandexhaust "DustTireBlurColor" (1.0 1.0 1.0 1.0)))
    (:else
      _))
  (Utilities|FlowControl|Sequence
    (:then_0
      (Utilities|Time|ClearandInvalidateTimerbyHandle (Variables|WaterInteraction|GetTimerSnapFoamToWaterSurface)))
    (:then_1
      (Utilities|IsValid _niagaracomp_inwater
        (:Is%20Valid
          (Components|Activation|Deactivate _niagaracomp_inwater))))
    (:then_2
      (CallFunction|ResetEnterWaterDoOnceServer)
      (if (not (CallFunction|IsOverlappingWaterBody))
        (Variables|WaterInteraction|SetWaterBody 0)))))

(event Custom|RadioGo
  (params (Start :type Boolean) (RadioStationIndex :type Integer))
  (bind _vehicleradio (Variables|Radio|GetVehicleRadio))
  ;[1 1 1] Manage Radio
  (if Start
    (Class|BPVehicleRadio|StartRadio _vehicleradio RadioStationIndex)
    (else
      (Class|BPVehicleRadio|StopRadio _vehicleradio))))

(event Custom|DestroyRadio
  (bind _vehicleradio (Variables|Radio|GetVehicleRadio))
  ;[1 1 1] Manage Radio
  (Class|BPVehicleRadio|StopRadio _vehicleradio)
  (Actor|DestroyActor _vehicleradio))

(event Custom|OnMIDCreated_Slot4
  (params (MID :type Material%20Instance%20Dynamic%20Object%20Reference))
  ;[1 1 1] Manage Local Passenger Material Swaps
  (Variables|MaterialsandTextures|SetMID_BodyLOD1 MID))

(event Custom|OnMIDCreated_Slot5
  (params (MID :type Material%20Instance%20Dynamic%20Object%20Reference))
  ;[1 1 1] Manage Local Passenger Material Swaps
  (Variables|MaterialsandTextures|SetMID_PartsLOD1 MID))

(event Custom|OnMIDCreated_Slot6
  (params (MID :type Material%20Instance%20Dynamic%20Object%20Reference))
  ;[1 1 1] Manage Local Passenger Material Swaps
  (Variables|MaterialsandTextures|SetMID_TurretLOD1 MID))

(event EventOnGearShiftDown
  (bind _ns_tank_roadandexhaust (Variables|Default|GetNS_Tank_RoadAndExhaust))
  ;[1 1 1] Exhaust Fire Blip
  (Niagara|SetNiagaraVariable(Bool) _ns_tank_roadandexhaust "IsShifting" true)
  (Utilities|FlowControl|Delay 0.125)
  (Niagara|SetNiagaraVariable(Bool) _ns_tank_roadandexhaust "IsShifting" false))

(event OnExitedWaterDelegate(PontoonsComponent)
  (params (Pontoon :type Spherical%20Pontoon%20Structure%20%28by%20ref%29))
  ;[0.6084 0.8133 1 1 33] EVENT -> Enter/Exit Water Body
  ;;[1 1 1] EVENT -> On Exited Water Body: Deactivate water FX When vehicle is out of water
  (CallFunction|ExitWaterBodyServer))

(event Custom|ExitWaterBodyServer
  ;[0.6084 0.8133 1 1 33] EVENT -> Enter/Exit Water Body
  ;;[1 1 1] EVENT -> On Exited Water Body: Deactivate water FX When vehicle is out of water
  (Variables|Default|SetNumPontoonInWater (Math|Integer|Clamp(Integer) (- (Variables|Default|GetNumPontoonInWater) 1) 0 5)))

(event Custom|CE_PawnExitVehicleWhileInWater
  (bind _niagaracomp_inwater (Variables|FX|GetNiagaraComp_InWater))
  ;[0.6084 0.8133 1 1 33] EVENT -> Enter/Exit Water Body
  ;;[1 1 1] EVENT -> On Exited Water Body: Deactivate water FX When vehicle is out of water
  (Utilities|IsValid _niagaracomp_inwater
    (:Is%20Valid
      (Niagara|SetNiagaraVariable(Float) _niagaracomp_inwater "TireSpinWaterSpawn" 0.0)
      (Components|Activation|Deactivate _niagaracomp_inwater))))

(event EventOnExitedWaterBody
  (params (WaterBody :type Fort%20Water%20Body%20Actor%20Object%20Reference))
  ;[0.6084 0.8133 1 1 33] EVENT -> Enter/Exit Water Body
  ;;[1 1 1] EVENT -> On Exited Water Body: Deactivate water FX When vehicle is out of water
  (|CallParentFunction WaterBody))

(event TreadVehicle|EventOnTreadCompressed
  (params (HitResult :type Hit%20Result%20Structure) (Compression :type Float%20%28single-precision%29) (CompressionVelocity :type Float%20%28single-precision%29))
  ;[1 1 1] EVENT -> OnTreadCompression - play force feedback / audio
  (Tracks|CompressionTracksAudio (Utilities|GetAudioController) HitResult Compression CompressionVelocity))

(event Custom|PawnEnteredVehicleSeat
  (params (Vehicle :type Fort%20Vehicle%20Interface%20Interface%20%28by%20ref%29) (PlayerPawn :type Fort%20Pawn%20Object%20Reference) (SeatIndex :type Integer))
  (bind _turretgunnerpawn (Variables|VehicleDefaults|GetTurretGunnerPawn))
  (bind _returnvalue (Pawn|GetController _turretgunnerpawn))
  ;[0 0.1562 0.01802] EVENT -> Pawn Enter Vehicle
  (bind _asfort_player_pawn (Utilities|Casting|CastToFortPlayerPawn PlayerPawn))
  (CallFunction|CEPawnEnterVehicleWhileinWater)
  (Variables|VehicleDefaults|SetisVehicleOccupied true)
  (Utilities|FlowControl|Sequence
    (:then_0
      (switch int SeatIndex
        (:Default
          _)
        (:0
          ;; ManageStateDriverEnter
          (bind _output_get (Variables|VehicleDefaults|SetDriverPawn _asfort_player_pawn))
          (Utilities|FlowControl|Sequence
            (:then_0
              (if (Variables|DriverCover|GetDriverCoverEnabled)
                (CallFunction|DriverCoverActivateonDriverEnter)
                (else
                  (Variables|DriverCover|SetDriverCoverOnCooldown true) ; This is to control the animation state
                ))
              (bind _asplayer_controller (Utilities|Casting|CastToPlayerController (Pawn|GetController _output_get)))
              (bind _output_get_2 (Variables|VehicleDefaults|SetDriverPlayerController _asplayer_controller))
              (CallFunction|CanHitLastDriver)
              (Variables|VehicleImpact|SetPassengerIsInstigator false)
              (bind _output_get_3 (Variables|Default|SetDriverPC (Fortnite|GetFortPlayerControllerfromActor _output_get_2)))
              (CallFunction|HideHudElements _output_get_3 true))
            (:then_1
              (Variables|Camera|SetCameraSpaceForwardDistanceOffset 0.0)
              (bind _restricted (CallFunction|IsDriverRestrictedbyTeamSetting _output_get))
              (if (not _restricted)
                (Class|BPTankAudioController|OnDriverStateChangeAudio (Utilities|GetAudioController) true)))
            (:then_2
              _)))
        (:1
          ;; ManageStatePassengerGunnerEnter
          (Variables|VehicleDefaults|SetTurretGunnerPawn _asfort_player_pawn)
          (Utilities|FlowControl|Sequence
            (:then_0
              (bind _asplayer_controller_2 (Utilities|Casting|CastToPlayerController _returnvalue))
              (Variables|VehicleDefaults|SetTurretGunnerPlayerController _asplayer_controller_2)
              (bind _output_get_4 (Variables|Default|SetGunnerPC (Fortnite|GetFortPlayerControllerfromActor _returnvalue)))
              (CallFunction|HideHudElements _output_get_4 true))
            (:then_1
              (Animation|Montage|PlayMontage (Variables|Vehicle|GetSkeletalMesh) (Variables|Animation|GetMontage_TurretEnter))
              (Animation|Montage|PlayMontage (Variables|/Script/Engine.Character|GetMesh _turretgunnerpawn) (Variables|Animation|GetMontage_GunnerEnter))
              (bind _returnvalue_2 (Utilities|Time|SetTimerbyEvent (delegate Custom|GetFrontTurretWeapon) 0.1 true)) ; Timer runs until Player and weapon are valid
              (Variables|VehicleDefaults|SetGetWeaponTimer _returnvalue_2))
            (:then_2
              (CallFunction|EnableGunnerHatchCollision true))))
        (:2
          _)
        (:3
          _))
      (CallFunction|SetLocalPassenger _asfort_player_pawn SeatIndex))
    (:then_1
      (if (not false)
        (Niagara|SetNiagaraVariable(Bool) (Variables|Default|GetNS_Tank_RoadAndExhaust) "VehicleHasDriver" true)))
    (:then_2
      (if (Networking|IsServer)
        (GameplayEffects|ApplyGameplayEffectToTarget :self (Ability|GetAbilitySystemComponent _asfort_player_pawn) :GameplayEffectClass (Variables|VehicleDefaults|GetGE_ApplyInTankTag) :Target (Ability|GetAbilitySystemComponent _asfort_player_pawn) :Level 1.0)
        (if (GameplayTags|HasAnyMatchingGameplayTags (Variables|VehicleDefaults|GetVehicleAllowFlippingTag))
          (GameplayEffects|ApplyGameplayEffectToTarget :self (Ability|GetAbilitySystemComponent _asfort_player_pawn) :GameplayEffectClass (Variables|VehicleDefaults|GetGE_ApplyAllowFlippingTag) :Target (Ability|GetAbilitySystemComponent _asfort_player_pawn) :Level 1.0))))))

(event TreadVehicle|EventOnTreadDecompressed
  (params (HitResult :type Hit%20Result%20Structure) (Compression :type Float%20%28single-precision%29) (CompressionVelocity :type Float%20%28single-precision%29))
  ;[1 1 1] EVENT -> OnTreadCompression - play force feedback / audio
  (Tracks|CompressionTracksAudio (Utilities|GetAudioController) HitResult Compression CompressionVelocity))

(event EventOnPawnEnterVehicle
  (params (PlayerPawn :type Fort%20Player%20Pawn%20Object%20Reference) (SeatIdx :type Integer))
  ;[0 0.1562 0.01802] EVENT -> Pawn Enter Vehicle
  (Ability|SendGameplayEventtoActor PlayerPawn "(TagName=\"Event.Vehicle.Entered\")" (Utilities|Struct|MakeGameplayEventData :Instigator self :OptionalObject self :ContextHandle "()" :InstigatorTags "(GameplayTags=)" :TargetTags "(GameplayTags=)" :EventMagnitude 0.0 :TargetData "()"))
  (Utilities|Array|AddUnique (Variables|VehicleDefaults|GetPassengers) PlayerPawn))

(event SurfaceType|EventOnSurfaceTypeVehicleIsOnChanged
  (params (SurfaceTypeVehicleIsOn :type EPhysicalSurface%20Enum :default "SurfaceType_Default"))
  ;[1 1 1] EVENT -> OnTreadCompression - play force feedback / audio
  (Variables|Audio|SetSurfaceChangeTime (Utilities|Time|GetGameTimeinSeconds)))

(event Custom|OverheatValueChanged
  (params (CurrentValue :type Float%20%28single-precision%29) (MaxValue :type Float%20%28single-precision%29) (State :type EFortWeaponOverheatState%20Enum) (bIsInCooldown :type Boolean))
  ;[1 1 1] EVENT -> Turret Gun Overheat Value Changed
  (CallFunction|MulticastOverheatValueChanged CurrentValue MaxValue State bIsInCooldown))

(event EventOnSleepStateChanged
  (params (bAwake :type Boolean))
  ;[1 1 1 1 25] Awake/Asleep
  (Utilities|FlowControl|Sequence
    (:then_0
      (if (Networking|IsDedicatedServer)
        _
        (elif bAwake
          (CallFunction|SetFxSettingsforVehicleType)
          (else
            (Components|Activation|Deactivate (Variables|Default|GetNS_Tank_RoadAndExhaust))))))
    (:then_1
      (Class|BPTankAudioController|OnSleepStateChange (Utilities|GetAudioController) bAwake))
    (:then_2
      ;;[1 1 1] If the Driver Cover or Engine are disabled then don't go to sleep!
      (if (or (Variables|DriverCover|GetDriverCoverOnCooldown) (Variables|TankEngine|GetEngineOverheated))
        (Vehicle|WakeUp)))))

(event EventOnHoldExitStarted
  (params (ExitingPawn :type Fort%20Pawn%20Object%20Reference) (ExitDuration :type Float%20%28single-precision%29))
  ;[1 1 1] EVENT -> OnHoldExitStarted/Stopped
  (Ability|SendGameplayEventtoActor ExitingPawn "(TagName=\"Athena.Vehicle.Biplane.OnHoldExitStarted\")" (Utilities|Struct|MakeGameplayEventData :ContextHandle "()" :InstigatorTags "(GameplayTags=)" :TargetTags "(GameplayTags=)" :EventMagnitude ExitDuration :TargetData "()")))

(event Custom|PawnExitVehicleSeat
  (params (Vehicle :type Fort%20Vehicle%20Interface%20Interface%20%28by%20ref%29) (PlayerPawn :type Fort%20Pawn%20Object%20Reference) (SeatIndex :type Integer))
  (bind _driverpc (Variables|Default|GetDriverPC))
  (bind _turretweapon (Variables|VehicleDefaults|GetTurretWeapon))
  (bind _gunnerpc (Variables|Default|GetGunnerPC))
  ;[0.9926 1 0] EVENT -> Pawn Exit Vehicle
  (bind _asfort_player_pawn (Utilities|Casting|CastToFortPlayerPawn PlayerPawn))
  (Utilities|FlowControl|Sequence
    (:then_0
      (Utilities|Array|RemoveItem (Variables|VehicleDefaults|GetPassengers) _asfort_player_pawn)
      (switch int SeatIndex
        (:Default
          _)
        (:0
          (Utilities|FlowControl|Sequence
            (:then_0
              (CallFunction|ClearLocalPassenger _asfort_player_pawn)
              ;; ManageStateDriverExit
              (Utilities|FlowControl|Sequence
                (:then_0
                  (Variables|VehicleImpact|SetIgnoreHitLastDriver true)
                  (Utilities|Time|SetTimerbyEvent (EventDispatchers|CreateEvent :!event CanHitLastDriver) 2.0)
                  (bind _restricted (CallFunction|IsDriverRestrictedbyTeamSetting _asfort_player_pawn))
                  (if (not _restricted)
                    (Class|BPTankAudioController|OnDriverStateChangeAudio (Utilities|GetAudioController) false))
                  (CallFunction|StopOutOfGasRumble)
                  (CallFunction|BoostRumbleStop)
                  (CallFunction|StopBoostCameraShake)
                  (CallFunction|DeactivateBoostFX)
                  (CallFunction|StopBoostRumble)
                  (CallFunction|CleanUpRoadandExhaustFX))
                (:then_1
                  (Utilities|BoostCameraBlur false
                    (:%20
                      (Rendering|Material|SetScalarParameterValueonMaterials (Variables|Vehicle|GetSkeletalMesh) "Tank On" 0.0) ; Turn off Brake Lights
                      (Utilities|IsValid _driverpc
                        (:Is%20Valid
                          (CallFunction|HideHudElements _driverpc)
                          (Variables|Default|SetDriverPC 0))
                        (:Is%20Not%20Valid
                          _))
                      (if (Variables|DriverCover|GetDriverCoverEnabled)
                        (if (Variables|DriverCover|GetDriverCoverOnCooldown)
                          (Variables|DriverCover|SetDriverCoverActiveOnDriverExit false)
                          (else
                            (Variables|DriverCover|SetDriverCoverActiveOnDriverExit true)
                            (CallFunction|RepairDamageablePart (Variables|DriverCover|GetDriverCoverDamageablePartName) 0.0) ; Setting Driver Cover Health to 0 deactivates collision
                          ))
                        (else
                          (Variables|DriverCover|SetDriverCoverOnCooldown false))))) ; Remove Bost blur on exit incase they were boosting while exiting
                )
                (:then_2
                  (Variables|VehicleDefaults|SetLastDriver _asfort_player_pawn)
                  (Variables|VehicleDefaults|SetDriverPawn 0)
                  (Variables|VehicleDefaults|SetDriverPlayerController 0))))
            (:then_1
              _)))
        (:1
          ;; ManageStateTurretGunnerExit
          (Utilities|FlowControl|Sequence
            (:then_0
              (Utilities|IsValid _gunnerpc
                (:Is%20Valid
                  (CallFunction|HideHudElements _gunnerpc)
                  (Variables|Default|SetGunnerPC 0))
                (:Is%20Not%20Valid
                  _))
              (CallFunction|EnableGunnerHatchCollision false)
              (Utilities|IsValid _turretweapon
                (:Is%20Valid
                  (RangedWeaponDual|UnbindEventfromOnHandFired _turretweapon (EventDispatchers|CreateEvent :!event TurretFireAnimations))
                  (Overheating|UnbindEventfromOnOverheatValueChanged _turretweapon (EventDispatchers|CreateEvent :!event Custom|OverheatValueChanged)))
                (:Is%20Not%20Valid
                  _))
              (Components|Activation|Deactivate (Variables|Default|GetNS_Tank_Turret_Overheat))
              (Animation|Montage|PlayMontage (Variables|Vehicle|GetSkeletalMesh) (Variables|Animation|GetMontage_TurretExit))
              (Animation|Montage|PlayMontage (Variables|/Script/Engine.Character|GetMesh _asfort_player_pawn) (Variables|Animation|GetMontage_GunnerExit)))
            (:then_1
              (Variables|VehicleDefaults|SetTurretGunnerPlayerController 0)
              (Variables|VehicleDefaults|SetTurretGunnerPawn 0)
              (Variables|VehicleDefaults|SetTurretWeapon 0)))))
      (if (not false)
        (Niagara|SetNiagaraVariable(Bool) (Variables|Default|GetNS_Tank_RoadAndExhaust) "VehicleHasDriver" false)))
    (:then_1
      (CallFunction|CEPawnExitVehicleWhileinWater))))

(event Custom|MulticastOverheatValueChanged
  (params (CurrentValue :type Float%20%28double-precision%29) (MaxValue :type Float%20%28double-precision%29) (OverheatState :type EFortWeaponOverheatState%20Enum) (IsInCooldown :type Boolean))
  ;[1 1 1] EVENT -> Turret Gun Overheat Value Changed
  (CallFunction|StartOverHeatChanges CurrentValue MaxValue OverheatState IsInCooldown))

(event EventOnHoldExitStopped
  (params (ExitingPawn :type Fort%20Pawn%20Object%20Reference))
  ;[1 1 1] EVENT -> OnHoldExitStarted/Stopped
  (Ability|SendGameplayEventtoActor ExitingPawn "(TagName=\"Athena.Vehicle.Biplane.OnHoldExitStopped\")"))

(event EventOnPawnExitVehicle
  (params (PlayerPawn :type Fort%20Player%20Pawn%20Object%20Reference) (ExitSocketName :type Name :default "None"))
  (bind _returnvalue (Ability|GetFortAbilitySystemComponent PlayerPawn))
  ;[0.9926 1 0] EVENT -> Pawn Exit Vehicle
  (CallFunction|ClearLocalPassenger PlayerPawn)
  (FortPawn|SetPawnVisibilityLocalControllerOnly PlayerPawn false true) ; Ensure pawn becomes visible locally again if exiting during cannon ADS
  (Variables|VehicleDefaults|SetisVehicleOccupied (!= (Utilities|Array|Length (Vehicle|GetAllPassengers)) 0))
  (if (Networking|IsServer)
    (GameplayEffects|RemoveActiveEffectswithGrantedTags _returnvalue (Variables|VehicleDefaults|GetInTankVehicleTag))
    (GameplayEffects|RemoveActiveGameplayEffectbySourceEffect _returnvalue (Variables|VehicleDefaults|GetGE_ApplyAllowFlippingTag) 0 -1)))

(event Custom|StartOverHeatChanges
  (params (CurrentValue :type Float%20%28double-precision%29) (MaxValue :type Float%20%28double-precision%29) (State :type EFortWeaponOverheatState%20Enum) (IsInCooldown :type Boolean))
  (bind _returnvalue (Math|Float|SafeDivide CurrentValue MaxValue))
  (bind _ns_tank_turret_overheat (Variables|Default|GetNS_Tank_Turret_Overheat))
  ;[1 1 1] EVENT -> Turret Gun Overheat Value Changed
  (if (not (Networking|IsServer))
    (switch Utilities|FlowControl|Switch|SwitchonEFortWeaponOverheatState State
      (:Heating
        (Class|FortVehicleInterface|CallWeaponFired(Message) self (Variables|Default|GetTurret_FortVehicleAimingWeaponComp) _returnvalue)
        (Utilities|IsValid _ns_tank_turret_overheat
          (:Is%20Valid
            (if (not (Components|Activation|IsActive _ns_tank_turret_overheat))
              (Components|Activation|Activate _ns_tank_turret_overheat false))
            (Niagara|SetNiagaraVariable(Float) _ns_tank_turret_overheat "User.OverheatAmount" _returnvalue))))
      (:Cooling
        (if (Variables|VehicleDefaults|GetTurretOverheated)
          (if (< _returnvalue 0.4)
            (Variables|VehicleDefaults|SetTurretOverheated false)
            (Niagara|SetNiagaraVariable(Float) _ns_tank_turret_overheat "User.StartCooling" 0.0) ; If we coolded down enough then start updating the overheat amount again
            (else
              _)))
        (Utilities|IsValid _ns_tank_turret_overheat
          (:Is%20Valid
            (Niagara|SetNiagaraVariable(Float) _ns_tank_turret_overheat "User.OverheatAmount" _returnvalue))))
      (:Overheated
        (Class|FortVehicleInterface|CallGunOverheat self (Variables|Default|GetTurret_FortVehicleAimingWeaponComp))
        (Utilities|IsValid _ns_tank_turret_overheat
          (:Is%20Valid
            (Niagara|SetNiagaraVariable(Float) _ns_tank_turret_overheat "User.OverheatAmount" 1.0)
            (Niagara|SetNiagaraVariable(Float) _ns_tank_turret_overheat "User.StartCooling" 1.0) ; If we overheated then start cooling
            (Variables|VehicleDefaults|SetTurretOverheated true)))))))

(event EventOnCollisionHitEffects
  (params (HitLocation :type Vector%20%28by%20ref%29 :default "0, 0, 0") (HitNormalImpulse :type Vector%20%28by%20ref%29 :default "0, 0, 0") (HitFrictionImpulse :type Vector%20%28by%20ref%29 :default "0, 0, 0") (HitNormal :type Vector%20%28by%20ref%29 :default "0, 0, 0") (HitActor :type Actor%20Object%20Reference) (HitSurfaceType :type EPhysicalSurface%20Enum :default "SurfaceType_Default"))
  (bind _returnvalue (Math|Vector|VectorLength HitNormalImpulse))
  (bind _returnvalue_2 (Math|Conversions|RotationFromXVector HitNormal))
  (bind _returnvalue_3 (Math|Rotator|InvertRotator _returnvalue_2))
  (bind (_asfort_pawn _bsuccess) (Utilities|Casting|CastToFortPawn HitActor))
  (bind _returnvalue_4 (/ _returnvalue 1e+02))
  ;[0 0 0] EVENT -> Collision Hit Effects:     Audio/FX Collision, used for skid fx and player hits
  (Class|BPTankAudioController|OnCollisionHit (Utilities|GetAudioController) HitActor HitLocation)
  (CallFunction|PlayHapticsonHit _returnvalue_4)
  (if (Math|Boolean|ANDBoolean (> _returnvalue_4 (Variables|VehicleImpact|GetScrapeImpact_Strength)) (>= (- (Utilities|Time|GetGameTimeinSeconds) (Variables|VehicleImpact|GetLastCollisionTime)) (Variables|VehicleImpact|GetMinTimeBetweenCollision)) (>= (- (Vehicle|GetTotalSpeedKmh) (Variables|VehicleImpact|GetSpawnImpactFX_VehicleSpeed)) (Variables|VehicleImpact|GetSpawnImpactFX_VehicleSpeed_Limit)))
    (Variables|VehicleImpact|SetLastCollisionTime (Utilities|Time|GetGameTimeinSeconds)) ; Setting Last Collidion Time Lowers Impact Amount
    (if _bsuccess
      ;;[1 1 1] Player Pawn impact
      (bind _returnvalue_5 (Effects|Components|ParticleSystem|SpawnFXAtLocation "/Valet/FX/Niagara/NS_Scrape_PlayerImpact.NS_Scrape_PlayerImpact" HitLocation _returnvalue_3))
      (Effects|Components|ParticleSystem|SetFloatParameter _returnvalue_5 "ImpactStrength" _returnvalue)
      (else
        ;;[1 1 1] World Impact, set to auto relase
        (bind _returnvalue_6 (Effects|Components|ParticleSystem|SpawnFXAtLocation "/Valet/FX/Niagara/NS_Scrape_Impact.NS_Scrape_Impact" HitLocation _returnvalue_3 "1.000000,1.000000,1.000000" true "AutoRelease"))
        (Effects|Components|ParticleSystem|SetFloatParameter _returnvalue_6 "ImpactStrength" _returnvalue)))))

(event EventOnBoostStarted
  ;[0.2027 0.7676 1] EVENT -> Boost
  (if (not (Vehicle|IsDriverSplitScreen))
    (Utilities|BoostCameraBlur true
      (:%20
        _)))
  (bind _drivercontroller (Utilities|IsValidDriverController)
    (:Is%20Valid
      (bind _returnvalue (CameraShakes|StartCameraShake :self (Variables|/Script/Engine.PlayerController|GetPlayerCameraManager _drivercontroller) :ShakeClass (Variables|CameraShakes|GetCameraShakeBoost) :Scale 1.0 :PlaySpace "CameraLocal"))
      (Variables|CameraShakes|SetBoostCameraShake (CameraShakes|ConvLegacyCameraShake _returnvalue))
      (CallFunction|BoostFOVRumble))
    (:Is%20Not%20Valid
      _))
  (CallFunction|ActivateTireSmoke)
  (CallFunction|ActivateBoostFX)
  (Class|BPTankAudioController|BoostStartAudio (Utilities|GetAudioController)))

(event Custom|BoostFOV_Rumble
  ;[0.2027 0.7676 1] EVENT -> Boost
  ;;[0 0 0] Boost FoV and Rumble
  (CallFunction|StartBoostRumble)
  (Variables|Boost|SetBoostCameraActive true)
  (Variables|Boost|SetisBoosting true)
  (bind (_direction _fov _rumble) (Timeline|BoostCameraTimeline Play)
    (:Update
      (Variables|Camera|SetCameraFOVOffset (Math|Float|Lerp 0.0 (Variables|Boost|GetMaxBoostFOV) _fov))
      (Variables|Boost|SetBoostRumbleIntensity _rumble)
      (CallFunction|UpdateBoostRumble))
    (:Finished
      (switch Utilities|FlowControl|Switch|SwitchonETimelineDirection _direction
        (:Backward
          (Variables|Boost|SetBoostCameraActive false)
          (CallFunction|StopBoostRumble)
          (Variables|Boost|SetisBoosting false))))))

(event Custom|StartBoostRumble
  ;[0.2027 0.7676 1] EVENT -> Boost
  ;;[0 0 0] Boost FoV and Rumble
  ;;;[1 1 1] Boost Rumble
  (Game|Feedback|PlayDynamicForceFeedback (Variables|VehicleDefaults|GetDriverPlayerController) (Variables|Boost|GetBoostRumbleIntensity) 3e+01 true false true false))

(event Custom|UpdateBoostRumble
  ;[0.2027 0.7676 1] EVENT -> Boost
  ;;[0 0 0] Boost FoV and Rumble
  ;;;[1 1 1] Boost Rumble
  (Game|Feedback|PlayDynamicForceFeedback (Variables|VehicleDefaults|GetDriverPlayerController) (Variables|Boost|GetBoostRumbleIntensity) 3e+01 true false true false))

(event Custom|ReverseBoostTimeline
  ;[0.2027 0.7676 1] EVENT -> Boost
  ;;[0 0 0] Boost FoV and Rumble
  (Timeline|BoostCameraTimeline Reverse))

(event EventPawnHitByVehicle
  (params (Pawn :type Fort%20Pawn%20Object%20Reference) (ImpactPoint :type Vector%20%28by%20ref%29 :default "0, 0, 0") (ImpactNormal :type Vector%20%28by%20ref%29 :default "0, 0, 0") (LaunchVector :type Vector :default "0, 0, 0") (bTeleportPawn :type Boolean) (TeleportPawnDistance :type Float%20%28single-precision%29))
  ;[1 1 1] EVENT -> Hit Pawn
  (bind _output_get (Variables|VehicleImpact|SetLaunchPlayerOnImpactVelocity LaunchVector))
  (CallFunction|OnPawnAudioCollision Pawn _output_get)
  (bind _asfort_player_pawn_athena (Utilities|Casting|CastToFortPlayerPawnAthena Pawn)
    (:then
      (Pawn|Character|LaunchCharacterJump _asfort_player_pawn_athena (Variables|VehicleImpact|GetLaunchPlayerOnImpactVelocity) false true (Team|OnSameTeam (Variables|VehicleDefaults|GetDriverPawn) _asfort_player_pawn_athena) true)
      (if (Variables|VehicleDefaults|GetbRagdollEnabled)
        (GameplayEffects|FortApplyGameplayEffectToTarget :self (Ability|GetFortAbilitySystemComponent (Variables|VehicleDefaults|GetDriverPawn)) :GameplayEffectClass "/Game/Abilities/Ragdoll/GE_Athena_Player_Ragdoll.GE_Athena_Player_Ragdoll_C" :TargetOfEffect (Ability|GetFortAbilitySystemComponent _asfort_player_pawn_athena) :Level 0.0))
      (for _array_element (Vehicle|GetAllPassengers)
        (GameplayCue|ExecuteGameplayCue%28Burst%29Local _array_element (Variables|GameplayCues|GetHitPlayerCue) (Variables|GameplayCues|GetGCParamsEmpty))
        (bind _output_get_2 (Variables|VehicleDefaults|SetLastHitPlayer _asfort_player_pawn_athena :!label _label))
        (CallFunction|SetFallInstigator _output_get_2))
      (Utilities|Time|SetTimerbyEvent (delegate Custom|ResetLastHitPlayer) 0.2)
      (goto _label))
    (:CastFailed
      (bind _asfort_aipawn (Utilities|Casting|CastToFortAIPawn Pawn))
      (Components|Movement|StopMovementImmediately (Variables|/Script/Engine.Character|GetCharacterMovement _asfort_aipawn))
      (Character|LaunchCharacter _asfort_aipawn (Variables|VehicleImpact|GetLaunchPlayerOnImpactVelocity) true true) ; Launch player based on cart velocity
      (Networking|SwitchHasAuthority
        (:Remote
          (for _array_element_2 (Vehicle|GetAllPassengers)
            (GameplayCue|ExecuteGameplayCue%28Burst%29Local _array_element_2 (Variables|GameplayCues|GetHitFiendCue) (Variables|GameplayCues|GetGCParamsEmpty))))))))

(event Custom|StopBoostRumble
  ;[0.2027 0.7676 1] EVENT -> Boost
  ;;[0 0 0] Boost FoV and Rumble
  ;;;[1 1 1] Boost Rumble
  (Game|Feedback|PlayDynamicForceFeedback (Variables|VehicleDefaults|GetDriverPlayerController) (Variables|Boost|GetBoostRumbleIntensity) 3e+01 true false true false))

(event EventOnBoostFinished
  ;[0.2027 0.7676 1] EVENT -> Boost
  (if (not (Vehicle|IsDriverSplitScreen))
    (Utilities|BoostCameraBlur false
      (:%20
        _)))
  (CallFunction|StopBoostCameraShake)
  (CallFunction|DeactivateTireSmoke)
  (CallFunction|DeactivateBoostFX)
  (Class|BPTankAudioController|BoostEndAudio (Utilities|GetAudioController)))

(event Custom|CleanupLoopingPooledFX_Nativize
  ;[0.1077 1 0] Clean up Looping Pooled FX
  (Utilities|FlowControl|Sequence
    (:then_0
      (bind _ns_tank_roadandexhaust (Variables|Default|GetNS_Tank_RoadAndExhaust)
        (:then
          (Components|Activation|Deactivate _ns_tank_roadandexhaust)
          (Variables|Default|SetNS_Tank_RoadAndExhaust 0))
        (:else)))
    (:then_1
      (bind _niagaracomp_tankdamage (Variables|FX|GetNiagaraComp_TankDamage)
        (:then
          (Components|Activation|Deactivate _niagaracomp_tankdamage)
          (Variables|FX|SetNiagaraComp_TankDamage 0)
          ;;[0.1077 1 0] Replace with Native Comp
          (Variables|Default|SetNativeComp_Damage 0))
        (:else)))
    (:then_2
      (bind _niagaracomp_enginedamage (Variables|Default|GetNiagaraComp_EngineDamage)
        (:then
          (Components|Activation|Deactivate _niagaracomp_enginedamage)
          (Variables|Default|SetNiagaraComp_EngineDamage 0))
        (:else)))
    (:then_3
      (bind _niagaracomp_lowfuelsputtersmoke (Variables|FX|GetNiagaraComp_LowFuelSputterSmoke)
        (:then
          (Components|Activation|Deactivate _niagaracomp_lowfuelsputtersmoke)
          (Variables|FX|SetNiagaraComp_LowFuelSputterSmoke 0))
        (:else)))
    (:then_4
      (bind _niagaracomp_treadskid (Variables|FX|GetNiagaraComp_TreadSkid)
        (:then
          (Components|Activation|Deactivate _niagaracomp_treadskid)
          (Variables|FX|SetNiagaraComp_TreadSkid 0))
        (:else)))
    (:then_5
      (bind _niagaracomp_boost (Variables|Default|GetNiagaraComp_Boost)
        (:then
          (Components|Activation|Deactivate _niagaracomp_boost)
          (Variables|Default|SetNiagaraComp_Boost 0))
        (:else)))
    (:then_6
      (bind _niagaracomp_inwater (Variables|FX|GetNiagaraComp_InWater)
        (:then
          (Components|Activation|Deactivate _niagaracomp_inwater)
          (Variables|FX|SetNiagaraComp_InWater 0))
        (:else)))))

(event Custom|BoostCameraReverse
  ;[0.2027 0.7676 1] EVENT -> Boost
  ;;[0 0 0] Boost FoV and Rumble
  (if (Variables|Boost|GetBoostCameraActive)
    (CallFunction|ReverseBoostTimeline)))

(event Custom|ResetLastHitPlayer
  ;[1 1 1] EVENT -> Hit Pawn
  (Variables|VehicleDefaults|SetLastHitPlayer 0))

(event Custom|BoostRumbleStop
  ;[0.2027 0.7676 1] EVENT -> Boost
  (CallFunction|StopBoostCameraShake))

(event Custom|OnPawnAudioCollision
  (params (HitPawn :type Actor%20Object%20Reference) (ImpactPoint :type Vector))
  ;[1 1 1] EVENT -> Hit Pawn
  (Class|BPTankAudioController|OnCollisionHit (Utilities|GetAudioController) HitPawn ImpactPoint))

(event Custom|StopBoostCameraShake
  ;[0.2027 0.7676 1] EVENT -> Boost
  (CallFunction|BoostCameraReverse)
  (Utilities|IsValid (Variables|CameraShakes|GetBoostCameraShake)
    (:Is%20Valid
      (bind _drivercontroller (Utilities|IsValidDriverController)
        (:Is%20Valid
          (CameraShakes|StopCameraShake (Variables|/Script/Engine.PlayerController|GetPlayerCameraManager _drivercontroller) (Variables|CameraShakes|GetBoostCameraShake) false)
          (Variables|CameraShakes|SetBoostCameraShake 0))))))

(event EventOnStopTick
  ;[0 0 0] Inform audio controller that Tick has stopped
  (Class|BPTankAudioController|OnTickStopAudio (Utilities|GetAudioController)))

(event EventOnStartFire
  (params (FireModeNum :type Byte))
  ;[1 1 1] EVENT -> Tank Catch on Fire
  (Variables|VehicleDefaults|SetbIsOnFire true))

(event Custom|ActivateBoostFX
  ;[0.2027 0.7676 1] EVENT -> Boost
  ;;[0.543 0.2454 1] Boost FX
  (bind _output_get (Variables|VehicleDefaults|SetBoostFX true))
  (Niagara|SetNiagaraVariable(Bool) (Variables|Default|GetNS_Tank_RoadAndExhaust) "BoostFX" _output_get))

(event Custom|DeactivateBoostFX
  ;[0.2027 0.7676 1] EVENT -> Boost
  ;;[0.543 0.2454 1] Boost FX
  (bind _output_get (Variables|VehicleDefaults|SetBoostFX false))
  (Niagara|SetNiagaraVariable(Bool) (Variables|Default|GetNS_Tank_RoadAndExhaust) "BoostFX" _output_get))

(event EventOnSetSeatCollision
  (params (SeatIndex :type Integer) (bOccupied :type Boolean))
  (bind _returnvalue (Utilities|Enum|LiteralenumECollisionEnabled "NoCollision"))
  (bind _returnvalue_2 (Utilities|Enum|LiteralenumECollisionEnabled "QueryOnly"))
  (bind _returnvalue_3 (select bOccupied _returnvalue_2 _returnvalue))
  ;[1 1 1] EVENT -> Toggle collision on passenger turret when Gunner Enters
  (bind _returnvalue_4 (Vehicle|GetVehicleSeatComponent))
  (switch name (Utilities|Struct|BreakAthenaCarPlayerSlot (Class|FortVehicleSeatComponent|GetPlayerSlot _returnvalue_4 SeatIndex))
    (:Passenger
      (Collision|SetCollisionEnabled (Variables|Default|GetTurretHatchCollision) _returnvalue_3)
      (Collision|SetCollisionEnabled (Variables|Default|GetTurretGunCollision) _returnvalue_3))))

(event EventOnFireStop
  ;[1 1 1] EVENT -> Tank no longer on Fire
  (Variables|VehicleDefaults|SetbIsOnFire false))

(event Custom|HealthChanged
  ;[1 1 1] EVENT -> On Damaged: Update material based on remaining health %
  (Variables|VehicleExplode|SetCanExplodeHack true)
  (if (not (Variables|VehicleExplode|GetVehicleExploded))
    (bind _output_get (Variables|VehicleDamageState|SetDamageValue (Math|Float|Lerp 1.0 0.0 (Math|Float|SafeDivide (Class|FortVehicleInterface|GetHealth) (Class|FortVehicleInterface|GetMaxHealth)))))
    (CallFunction|VFXDamageState _output_get)
    (CallFunction|UpdateDamageState _output_get)))

(event Custom|Update%20Damage%20State
  (params (Damage :type Float%20%28double-precision%29))
  ;[1 1 1] EVENT -> On Damaged: Update material based on remaining health %
  (CallFunction|SendDamageValuetoMaterials Damage))

(event EventOnDamaged
  (params (Damage :type Float%20%28single-precision%29) (DamageTags :type Gameplay%20Tag%20Container%20Structure%20%28by%20ref%29) (EffectContext :type Gameplay%20Effect%20Context%20Handle%20Structure%20%28by%20ref%29) (EventInstigator :type Controller%20Object%20Reference) (DamageCauser :type Actor%20Object%20Reference))
  ;[1 1 1] EVENT -> On Damaged: Update material based on remaining health %
  (bind _asfort_player_controller (Utilities|Casting|CastToFortPlayerController EventInstigator))
  (Variables|VehicleExplode|SetExplodeInstigator (Game|Player|GetPlayerPawn _asfort_player_controller))
  (CallFunction|EventonDamageThreshold DamageCauser (Ability|EffectContext|GetHitResult EffectContext) _asfort_player_controller))

(event Custom|VFX_DamageState
  (params (Damage :type Float%20%28double-precision%29))
  (bind _niagaracomp_tankdamage (Variables|FX|GetNiagaraComp_TankDamage))
  (bind _niagaracomp_tankdamage_2 (Variables|FX|GetNiagaraComp_TankDamage))
  ;[1 1 1] EVENT -> Turn on damage state Niagara vfx
  (Class|BPTankAudioController|OnDamageAudio (Utilities|GetAudioController) Damage)
  (if (>= Damage (Variables|FX|NiagaraSettings|GetDamage_Stage1))
    (Utilities|IsValid _niagaracomp_tankdamage_2
      (:Is%20Valid
        _)
      (:Is%20Not%20Valid
        (bind _returnvalue (Niagara|SpawnSystemAttached "/Tank/FX/NS_Tank_Damage.NS_Tank_Damage" (Variables|Vehicle|GetSkeletalMesh) "frame" "0, 0, 0" "0, 0, 0" "KeepRelativeOffset" false true "AutoRelease"))
        (bind _output_get (Variables|FX|SetNiagaraComp_TankDamage _returnvalue))
        ;;[0.1077 1 0] Replace with Native Comp
        (Variables|Default|SetNativeComp_Damage _output_get)))
    (Niagara|SetNiagaraVariable(Float) _niagaracomp_tankdamage_2 "User.DamageScale" Damage)
    (else
      (Utilities|IsValid _niagaracomp_tankdamage
        (:Is%20Valid
          (Niagara|SetNiagaraVariable(Float) _niagaracomp_tankdamage "User.DamageScale" Damage))) ; Still need to scale FX if it's active and we are healed beyond min threshold
    )))

(event TreadVehicle|EventOnTreadEnabled
  (params (bEnabled :type Boolean) (Side :type ETreadSide%20Enum :default "Invalid"))
  ;[1 1 1] EVENT -> On Tread Enabled/Disabled
  (for _array_element (Variables|VehicleDefaults|GetPassengers)
    (Utilities|IsValid _array_element
      (:Is%20Valid
        (Utilities|IsValid (Pawn|GetController _array_element)
          (:Is%20Valid
            (AI|Components|Blackboard|SetValueasBool (AI|GetBlackboard (Pawn|GetController _array_element)) (Utilities|String|StringToName "AwardThanks_TankTreadDestroyed") (not (and (TreadVehicle|IsRightTreadEnabled) (TreadVehicle|IsLeftTreadEnabled))))))))))

(event Custom|CleanUpRoadAndExhaustFX
  (bind _ns_tank_roadandexhaust (Variables|Default|GetNS_Tank_RoadAndExhaust))
  ;[1 1 1] Deactivate Road and Exhaust FX
  (Utilities|IsValid _ns_tank_roadandexhaust
    (:Is%20Valid
      (Components|Activation|Deactivate _ns_tank_roadandexhaust))))

(fn ActivateTireSmoke)

(fn ApplyBuildingExplosionDamage
  (bind _explodeinstigator (Variables|VehicleExplode|GetExplodeInstigator))
  (bind _returnvalue (Ability|GetFortAbilitySystemComponent _explodeinstigator))
  (bind _outhits (Collision|MultiSphereTraceForObjects (Transformation|GetActorLocation) (+ (Transformation|GetActorLocation) (Math|Vector|GetUpVector "0, 0, 0")) (Variables|VehicleExplode|GetExplodeRadius) (Variables|VehicleExplode|GetObjectArray)))
  (for _array_element _outhits
    (bind _returnvalue_2 (Ability|GameplayEffect|AssignTagSetbyCallerMagnitude (GameplayEffects|MakeOutgoingSpec :self _returnvalue :GameplayEffectClass (Variables|DamageGameplayEffects|GetGE_ExplodeBuildingDamage) :Level 1.0) "(TagName=\"SetByCaller.Damage\")" (Variables|VehicleExplode|GetCachedExplosionDamageBuildings)))
    (GameplayEffects|FortApplyGameplayEffectSpecToTarget :self _returnvalue :GameplayEffectSpecHandle _returnvalue_2 :TargetOfEffect (Ability|GetFortAbilitySystemComponent (.HitActor (Collision|BreakHitResult _array_element))) :OptionalSourceObject self :OptionalEffectCauser self :OptionalDamageSource self)))

(fn CacheExplosionDamageValues
  ;[1 0.5723 0.8743] Cache the explosion damage values from the default values (which are hooked up to data tables to be hotfixable). Cached values may be overridden by Creative Mode runtime user editable options.
  (Variables|VehicleExplode|SetCachedExplosionDamageBuildings (ScalableFloat|GetValueatLevel (Variables|VehicleExplode|GetExplosionDamageBuildings)))
  (Variables|VehicleExplode|SetCachedExplosionDamagePlayers (ScalableFloat|GetValueatLevel (Variables|VehicleExplode|GetExplosionDamagePlayers)))
  (Variables|VehicleExplode|SetCachedExplosionDamageVehicles (ScalableFloat|GetValueatLevel (Variables|VehicleExplode|GetExplosionDamageVehicles))))

(fn CalcImpulseDirection
  (params (Vehicle :type Actor%20Object%20Reference) (Magnitude :type Float%20%28double-precision%29))
  (return (Math|Vector|MakeVector (.x (* (Math|Vector|Normalize (- (Transformation|GetActorLocation Vehicle) (Transformation|GetActorLocation))) Magnitude)) (.y (* (Math|Vector|Normalize (- (Transformation|GetActorLocation Vehicle) (Transformation|GetActorLocation))) Magnitude)) (Math|Float|Max(Float) (+ (.z (* (Math|Vector|Normalize (- (Transformation|GetActorLocation Vehicle) (Transformation|GetActorLocation))) Magnitude)) 1.0) (Variables|VehicleExplode|GetExplodeImpulseZ_Magnitude)))))

(fn CalculateSpinThrustFromCannonFire
  (params (SpinThrustMagnitude :type Float%20%28double-precision%29) (Projectile :type Actor%20Object%20Reference))
  (return (* (Math|Vector|MirrorVectorbyNormal (Transformation|GetRightVector (Variables|Transformation|GetRootComponent Projectile)) (Transformation|GetRightVector (Variables|Transformation|GetRootComponent Projectile))) SpinThrustMagnitude)))

(fn CannonFiredReticle
  (bind _dynamicreticle_cannon (Variables|Default|GetDynamicReticle_Cannon))
  (bind _returnvalue (UserInterface|GetWidget _dynamicreticle_cannon))
  (bind _self self)
  (Networking|SwitchHasAuthority
    (:Remote
      (Utilities|IsValid _self
        (:Is%20Valid
          (bind _returnvalue_2 (VehicleWeapon|GetSeatWeaponComponent(Message) _self 0))
          (Utilities|IsValid _returnvalue_2
            (:Is%20Valid
              (Utilities|IsValid (Variables|/Script/FortniteGame.FortVehicleSeatWeaponComponent|GetCachedWeapon _returnvalue_2)
                (:Is%20Valid
                  (Utilities|IsValid _returnvalue
                    (:Is%20Valid
                      (bind _astank_vehicle_dynamic_reticle_widget (Utilities|Casting|CastToTankVehicle_DynamicReticleWidget _returnvalue))
                      (Class|TankVehicleDynamicReticleWidget|OnCannonFired _astank_vehicle_dynamic_reticle_widget (Math|Float|SafeDivide 1.0 (FortWeapon|GetFiringRate (Variables|/Script/FortniteGame.FortVehicleSeatWeaponComponent|GetCachedWeapon _returnvalue_2)))))))))))))))

(fn ClearAndInvalidateAllTimers
  (bind _engineoverheatdurationtimer (Variables|TankEngine|GetEngineOverheatDurationTimer))
  (bind _tankenginerepairtimer (Variables|TankEngine|GetTankEngineRepairTimer))
  (bind _drivercoverresetcooldowntimer (Variables|DriverCover|GetDriverCoverResetCooldownTimer))
  (bind _tankenginerepairdelaytimer (Variables|TankEngine|GetTankEngineRepairDelayTimer))
  (if (Utilities|Time|IsValidTimerHandle _tankenginerepairdelaytimer)
    (Utilities|Time|ClearandInvalidateTimerbyHandle _tankenginerepairdelaytimer) ; If already running when we take damage, reset the timer
  )
  (if (Utilities|Time|IsValidTimerHandle _drivercoverresetcooldowntimer)
    (Utilities|Time|ClearandInvalidateTimerbyHandle _drivercoverresetcooldowntimer) ; If already running when we take damage, reset the timer
  )
  (if (Utilities|Time|IsValidTimerHandle _tankenginerepairtimer)
    (Utilities|Time|ClearandInvalidateTimerbyHandle _tankenginerepairtimer) ; If already running when we take damage, reset the timer
  )
  (if (Utilities|Time|IsValidTimerHandle _engineoverheatdurationtimer)
    (Utilities|Time|ClearandInvalidateTimerbyHandle _engineoverheatdurationtimer) ; If already running when we take damage, reset the timer
  )
  (if (Utilities|Time|IsValidTimerHandle (Variables|DriverCover|GetDriverCoverResetTimer))
    _))

(fn ClearLocalPassenger
  (params (Player :type Fort%20Pawn%20Object%20Reference))
  (if (Pawn|IsLocallyViewed Player)
    (Utilities|FlowControl|Sequence
      (:then_0
        (bind _output_get (Variables|VehicleDefaults|SetIsLocalPlayerPassenger false)) ; Clear local passenger to swap back to Non Masked Materials
        (CallFunction|LocalPassengerMatSwap)
        (Collision|SetCollisionResponsetoChannel (Variables|Vehicle|GetSkeletalMesh) "ECC_Camera" "ECR_Block")
        (Class|BPTankAudioController|OnPassengerStateChangeAudio (Utilities|GetAudioController) _output_get))
      (:then_1
        (FortPawn|UnbindEventfromOnWeaponEquippedDelegate Player (EventDispatchers|CreateEvent :!event Passenger%20Equip%20Weapon))
        (bind _asb_ranged_generic (Utilities|Casting|CastToB_Ranged_Generic (Variables|/Script/FortniteGame.FortPawn|GetCurrentWeapon Player)))
        (Default|UnbindEventfromOnAimDownSightsChanged _asb_ranged_generic (EventDispatchers|CreateEvent :!event LocalPassengerADS) :!class /Game/Weapons/Blueprints/B_Ranged_Generic.B_Ranged_Generic_C)))))

(fn CreateBaseMID
  (bind _returnvalue (Rendering|Material|CreateDynamicMaterialInstance 0 "TankMID" "None" :!class /Script/Engine.KismetMaterialLibrary))
  (bind _output_get (Variables|MaterialsandTextures|SetMID _returnvalue))
  (Utilities|FlowControl|Sequence
    (:then_0
      (bind _returnvalue_2 (Rendering|Material|GetTextureParameterValue _output_get "Diffuse" :!class /Script/Engine.MaterialInstanceDynamic))
      (Variables|MaterialsandTextures|SetDiffuse _returnvalue_2))
    (:then_1
      (bind _returnvalue_3 (Rendering|Material|GetTextureParameterValue _output_get "Mask" :!class /Script/Engine.MaterialInstanceDynamic))
      (Variables|MaterialsandTextures|SetMask _returnvalue_3))
    (:then_2
      (bind _returnvalue_4 (Rendering|Material|GetTextureParameterValue _output_get "Normal" :!class /Script/Engine.MaterialInstanceDynamic))
      (Variables|MaterialsandTextures|SetNormal _returnvalue_4))
    (:then_3
      (bind _returnvalue_5 (Rendering|Material|GetTextureParameterValue _output_get "Scratch/Grime/EMPTY" :!class /Script/Engine.MaterialInstanceDynamic))
      (Variables|MaterialsandTextures|SetScratchGrime _returnvalue_5))
    (:then_4
      (bind _returnvalue_6 (Rendering|Material|GetTextureParameterValue _output_get "SpecularMask" :!class /Script/Engine.MaterialInstanceDynamic))
      (Variables|MaterialsandTextures|SetSpecularMask _returnvalue_6))
    (:then_5
      (bind _returnvalue_7 (Rendering|Material|GetVectorParameterValue _output_get "Layer 01 Color" :!class /Script/Engine.MaterialInstanceDynamic))
      (Variables|MaterialsandTextures|SetPrimaryColor _returnvalue_7))))

(fn CreateDestroyedMIDApply
  (bind _returnvalue (Rendering|Material|CreateDynamicMaterialInstance "/Valet/Materials/M_FN_Valet_Destroyed_Master_InstAdjusted.M_FN_Valet_Destroyed_Master_InstAdjusted" "ValetDestroyedMID" "None" :!class /Script/Engine.KismetMaterialLibrary))
  (bind _output_get (Variables|MaterialsandTextures|SetDestroyedMID _returnvalue))
  (Utilities|FlowControl|Sequence
    (:then_0
      (Rendering|Material|SetTextureParameterValue _output_get "Diffuse" (Variables|MaterialsandTextures|GetDiffuse))
      (Rendering|Material|SetMaterial (Variables|Vehicle|GetSkeletalMesh) 0 (Variables|MaterialsandTextures|GetDestroyedMID)))
    (:then_1
      (Rendering|Material|SetTextureParameterValue _output_get "Normal" (Variables|MaterialsandTextures|GetNormal)))
    (:then_2
      (Rendering|Material|SetTextureParameterValue _output_get "Mask" (Variables|MaterialsandTextures|GetMask)))
    (:then_3
      (Rendering|Material|SetTextureParameterValue _output_get "Scratch/Grime/EMPTY" (Variables|MaterialsandTextures|GetScratchGrime)))
    (:then_4
      (Rendering|Material|SetTextureParameterValue _output_get "Specular Mask" (Variables|MaterialsandTextures|GetSpecularMask)))
    (:then_5
      (Rendering|Material|SetVectorParameterValue _output_get "Layer 01 Color" (Variables|MaterialsandTextures|GetPrimaryColor) :!class /Script/Engine.MaterialInstanceDynamic))))

(fn DamageablePartChanged
  (params (PartConfig :type Vehicle%20Damageable%20Part%20Config%20Structure) (Part :type Vehicle%20Damageable%20Part%20Structure) (NewlyDisabled :type Boolean))
  (bind (_configindex _bonename _shapename _health) (Utilities|Struct|BreakVehicleDamageablePart Part))
  (bind (_shapename_2 _bonename_2) (Utilities|Struct|BreakVehicleDamageablePartConfig PartConfig))
  (bind (_returnvalue_location _returnvalue_rotation) (Math|Transform|BreakTransform (Transformation|GetSocketTransform (Variables|Vehicle|GetSkeletalMesh) "shatters_01" "RTS_World")))
  (bind (_returnvalue_location_2 _returnvalue_rotation_2) (Math|Transform|BreakTransform (Transformation|GetSocketTransform (Variables|Vehicle|GetSkeletalMesh) "road_wheel_05_r" "RTS_World")))
  (bind (_returnvalue_location_3 _returnvalue_rotation_3) (Math|Transform|BreakTransform (Transformation|GetSocketTransform (Variables|Vehicle|GetSkeletalMesh) "road_wheel_05_l" "RTS_World")))
  (bind _audiocontroller (Utilities|GetAudioController))
  (bind _ns_tank_roadandexhaust (Variables|Default|GetNS_Tank_RoadAndExhaust))
  (bind _ns_tank_roadandexhaust_2 (Variables|Default|GetNS_Tank_RoadAndExhaust))
  (bind _niagarafx_treadrepair (Variables|FX|GetNiagaraFX_TreadRepair))
  (bind _skeletalmesh (Variables|Vehicle|GetSkeletalMesh))
  (if NewlyDisabled
    (LocalVariables|SetBoneNameLocal _bonename_2)
    (switch name _shapename_2
      (:Tread_L
        ;[1 1 1] Tank Treads Destroyed
        (TreadVehicle|SetLeftTreadEnabled false)
        (bind _returnvalue (Niagara|SpawnSystematLocation "/Tank/FX/NS_Tank_Tread_Destroy.NS_Tank_Tread_Destroy" _returnvalue_location_3 _returnvalue_rotation_3))
        (Niagara|SetNiagaraVariable(Int32) _returnvalue "TankBone" 0)
        (Niagara|SetNiagaraVariable(Bool) _ns_tank_roadandexhaust_2 "Sparks_Wheel_FL" true)
        (Variables|FX|NiagaraSettings|SetTread_Destroyed_Left true)
        (Niagara|SetNiagaraVariable(Bool) _ns_tank_roadandexhaust_2 "TireIntactFL" false)
        (Class|BPTankAudioController|TracksTirePopAudio _audiocontroller false))
      (:Tread_R
        (TreadVehicle|SetRightTreadEnabled false)
        (bind _returnvalue_2 (Niagara|SpawnSystematLocation "/Tank/FX/NS_Tank_Tread_Destroy.NS_Tank_Tread_Destroy" _returnvalue_location_2 _returnvalue_rotation_2))
        (Niagara|SetNiagaraVariable(Int32) _returnvalue_2 "TankBone" 1)
        (Niagara|SetNiagaraVariable(Bool) _ns_tank_roadandexhaust_2 "Sparks_Wheel_FR" true)
        (Variables|FX|NiagaraSettings|SetTread_Destroyed_Right true)
        (Niagara|SetNiagaraVariable(Bool) _ns_tank_roadandexhaust_2 "TireIntactFR" false)
        (Class|BPTankAudioController|TracksTirePopAudio _audiocontroller true))
      (:DriverCover
        (if (Variables|DriverCover|GetDriverCoverEnabled)
          (if (Variables|DriverCover|GetDriverCoverActiveOnDriverExit)
            (if (Networking|IsServer)
              (CallFunction|DriverCoverDeactivateonDriverExit))
            (else
              (Variables|DriverCover|SetDriverHatchCurrentHealthPercent 0.0)
              (Niagara|SpawnSystematLocation "/Tank/FX/NS_Tank_Canopy_Destroy.NS_Tank_Canopy_Destroy" (+ _returnvalue_location "0,0,20.000000") _returnvalue_rotation)
              ;[1 1 1] Driver Cover Destroyed
              (if (Networking|IsServer)
                (CallFunction|DriverCoverDeactivated))))))
      (:Engine
        ;[1 1 1] Engine Destroyed
        (CallFunction|EventonEngineOverheat)
        (Niagara|SpawnSystemAttached (Variables|FX|GetNiagaraFX_EngineDestroyed) (Variables|Vehicle|GetSkeletalMesh) "Engine" "0, 0, 0" "0, 0, 0" "KeepRelativeOffset" true true "AutoRelease")
        (Niagara|SpawnSystemAttached (Variables|FX|GetNiagaraFX_EngineImpact) (Variables|Vehicle|GetSkeletalMesh) "Engine" "0, 0, 0" "0, 0, 0" "KeepRelativeOffset" true true "AutoRelease") ; Engine Hit FX
      ))
    (else
      ;[1 1 1] When treads are repaired set them to enabled so the interact prompt will appear
      (switch name _shapename_2
        (:Tread_L
          (if (not (TreadVehicle|IsLeftTreadEnabled))
            (TreadVehicle|SetLeftTreadEnabled true)
            (bind _returnvalue_3 (Niagara|SpawnSystemAttached _niagarafx_treadrepair _skeletalmesh "frame"))
            (Niagara|SetNiagaraVariable(Int32) _returnvalue_3 "TankBone" 0)
            (Niagara|SetNiagaraVariable(Bool) _ns_tank_roadandexhaust "TireIntactFL" true)))
        (:Tread_R
          (if (not (TreadVehicle|IsRightTreadEnabled))
            (TreadVehicle|SetRightTreadEnabled true)
            (bind _returnvalue_4 (Niagara|SpawnSystemAttached _niagarafx_treadrepair _skeletalmesh "frame"))
            (Niagara|SetNiagaraVariable(Int32) _returnvalue_4 "TankBone" 1)
            (Niagara|SetNiagaraVariable(Bool) _ns_tank_roadandexhaust "TireIntactFR" true)))
        (:Engine
          ;[1 1 1] Update Engine Damage State
          (if (Networking|IsServer)
            (Variables|TankEngine|SetTankEngineCurrent%20Health _health)
            (Variables|TankEngine|SetTankEngineCurrentHealthPercent (Math|Float|SafeDivide _health (Variables|TankEngine|GetEngineMaxHealth)))
            (CallFunction|TankEngineHealthChanged _health)))
        (:DriverCover
          ;[1 1 1] Set Driver Cover Current Health Percent
          (Variables|DriverCover|SetDriverHatchCurrentHealthPercent (Math|Float|SafeDivide _health (Variables|DriverCover|GetDriverCoverMaxHealth))))))))

(fn DamgeStateSetAmount
  (bind _niagaracomp_tankdamage (Variables|FX|GetNiagaraComp_TankDamage))
  (Niagara|SetNiagaraVariable(Float) _niagaracomp_tankdamage "DamageScale" (Variables|FX|NiagaraSettings|GetDamage_Stage1))
  (Niagara|SetNiagaraVariable(Float) _niagaracomp_tankdamage "DamageScale" (Variables|FX|NiagaraSettings|GetDamage_Stage2))
  (Niagara|SetNiagaraVariable(Float) _niagaracomp_tankdamage "DamageScale" (Variables|FX|NiagaraSettings|GetDamage_Stage3Critical)))

(fn DeactivateTireSmoke)

(fn DebugCannonAim
  (params (DrawCameraDebugLineFromDriverCamera :type Boolean))
  (Utilities|IsValid (Variables|VehicleDefaults|GetDriverPlayerController)
    (:Is%20Valid
      (Utilities|FlowControl|Sequence
        (:then_0
          (Networking|SwitchHasAuthority
            (:Remote
              (Rendering|Debug|DrawDebugLine (Transformation|GetSocketLocation (Variables|Vehicle|GetSkeletalMesh) "Muzzle") (+ (Utilities|Operators|Multiply (Math|Vector|GetUnitDirection(Vector) (Transformation|GetSocketLocation (Variables|Vehicle|GetSkeletalMesh) "CannonPivot") (Transformation|GetSocketLocation (Variables|Vehicle|GetSkeletalMesh) "Muzzle")) (Math|Vector|MakeVector (LocalVariables|GetLineTraceDistance) (LocalVariables|GetLineTraceDistance) (LocalVariables|GetLineTraceDistance))) (Transformation|GetSocketLocation (Variables|Vehicle|GetSkeletalMesh) "Muzzle")) "(R=1.000000,G=0.018814,B=0.000000,A=1.000000)" 0.2 1e+01)
              (if DrawCameraDebugLineFromDriverCamera
                (Rendering|Debug|DrawDebugLine (Camera|GetCameraLocation (Variables|/Script/Engine.PlayerController|GetPlayerCameraManager (Variables|VehicleDefaults|GetDriverPlayerController))) (+ (Camera|GetCameraLocation (Variables|/Script/Engine.PlayerController|GetPlayerCameraManager (Variables|VehicleDefaults|GetDriverPlayerController))) (* (Camera|GetCameraRotation (Variables|/Script/Engine.PlayerController|GetPlayerCameraManager (Variables|VehicleDefaults|GetDriverPlayerController))) (+ (LocalVariables|GetLineTraceDistance) 2e+03))) "(R=0.107666,G=1.000000,B=0.000000,A=1.000000)" 0.2 8.0)))))
        (:then_1
          _)))
    (:Is%20Not%20Valid
      _))
  (return))

(fn DestroyBuildingsTraceDirection
  (params (TraceRange :type Float%20%28double-precision%29))
  (return (+ (Transformation|GetWorldLocation (Variables|Default|GetOverlapVolume)) "0, 0,55.000000") (+ (+ (Transformation|GetWorldLocation (Variables|Default|GetOverlapVolume)) (* (Transformation|GetActorForwardVector) TraceRange)) "0, 0,55.000000") (Transformation|GetActorRotation)))

(fn DestroyVehiclePartsWhenDrivenOver
  (return (> (ScalableFloat|GetValueatLevel (Variables|VehicleDefaults|GetDestroyVehiclePartsOnDriveOver)) 0)))

(fn GetAnimData
  (return (Variables|TankEngine|GetEngineOverheated) (Variables|DriverCover|GetDriverCoverOnCooldown) (Variables|TankEngine|GetTankEngineCurrentHealthPercent) (Variables|DriverCover|GetDriverHatchCurrentHealthPercent)))

(fn GetTurnSlowMagnitude
  (params (SteeringAngle :type Float%20%28double-precision%29))
  (return (Math|Float|MapRangeClamped (Math|Float|Clamp(Float) (Math|Float|Absolute(Float) SteeringAngle) 5.0 45.0) 5.0 45.0 0.9 0.2)))

(fn HandleRefueledFromEmpty
  (bind _output_get (Variables|Fuel|SetVehicleHasFuel true))
  (Niagara|SetNiagaraVariable(Bool) (Variables|Default|GetNS_Tank_RoadAndExhaust) "VehicleHasFuel" _output_get))

(fn HideHudElements
  (params (InPlayerController :type Fort%20Player%20Controller%20Object%20Reference) (ShouldHide :type Boolean))
  (HUD|SetHUDElementVisibilitywithContext InPlayerController (Variables|VehicleDefaults|GetHideHUD_Tag) ShouldHide "HideHudElements" ""))

(fn IsDriverRestrictedByTeamSetting
  (params (PlayerPawn :type Fort%20Player%20Pawn%20Object%20Reference))
  ;[0.8734 0.03282 1] Is the driver restricted by team setting
  (return (CallFunction|IsDriverMinigameTeamRestricted PlayerPawn)))

(fn IsVehicleAccelerating
  (params (MaxSpeed :type Float%20%28double-precision%29))
  (return (Math|Boolean|ANDBoolean (Math|Float|InRange(Float) (Vehicle|GetForwardSpeedKmh) 2.0 MaxSpeed) (> (Vehicle|GetForwardSpeedKmh) (Vehicle|GetPreviousForwardSpeedKmh)) (Vehicle|PlayerIsTryingtoMoveForward) (Variables|Fuel|GetVehicleHasFuel) (not (Vehicle|VehicleIsinAir)))))

(fn OnRep_DriverCoverActiveOnDriverExit
  (if (Variables|DriverCover|GetDriverCoverActiveOnDriverExit)
    _))

(fn OnRep_DriverCoverMaxHealth)

(fn OnRep_DriverCoverOnCooldown)

(fn OnRep_DriverCoverResetCooldownTime)

(fn OnRep_DriverHatchCurrentHealthPercent)

(fn OnRep_EngineMaxHealth)

(fn OnRep_EngineOverheatDuration)

(fn OnRep_EngineOverheated)

(fn OnRep_NiagaraComp_EngineDamage)

(fn OnRep_NumPontoonInWater
  (if (> (Variables|Default|GetNumPontoonInWater) 0)
    (CallFunction|EnterWaterBody)
    (else
      (CallFunction|ExitWaterBody) ; Need this to be replicated or other clients get inconsistent results
    )))

(fn OnRep_TankEngineCurrentHealth)

(fn OnRep_TankEngineCurrentHealthPercent)

(fn PassengerEquipWeapon
  (params (NewWeapon :type Fort%20Weapon%20Object%20Reference) (PrevWeapon :type Fort%20Weapon%20Object%20Reference))
  (Utilities|FlowControl|Sequence
    (:then_0
      (bind _asb_ranged_generic (Utilities|Casting|CastToB_Ranged_Generic PrevWeapon))
      (Default|UnbindEventfromOnAimDownSightsChanged _asb_ranged_generic (EventDispatchers|CreateEvent :!event LocalPassengerADS) :!class /Game/Weapons/Blueprints/B_Ranged_Generic.B_Ranged_Generic_C))
    (:then_1
      (bind _asb_ranged_generic_2 (Utilities|Casting|CastToB_Ranged_Generic NewWeapon))
      (Default|BindEventtoOnAimDownSightsChanged _asb_ranged_generic_2 (EventDispatchers|CreateEvent :!event LocalPassengerADS) :!class /Game/Weapons/Blueprints/B_Ranged_Generic.B_Ranged_Generic_C)
      (Development|PrintString "Passenger ADS Bound"))))

(fn PlayHapticsOnHit
  (params (HitMagnitude :type Float%20%28double-precision%29))
  (bind _validplayercontrollers (LocalVariables|GetValidPlayerControllers))
  (Utilities|FlowControl|Sequence
    (:then_0
      (bind _drivercontroller (Utilities|IsValidDriverController)
        (:Is%20Valid
          (Utilities|Array|Add (LocalVariables|GetValidPlayerControllers) _drivercontroller))))
    (:then_1
      (if (> (Utilities|Array|Length (LocalVariables|GetValidPlayerControllers)) 0)
        (if (> (Variables|GetHitMagnitude) 4e+03)
          (Utilities|FlowControl|Sequence
            (:then_0
              (LocalVariables|SetCameraShakeMagnitude (Math|Float|MapRangeClamped (Variables|GetHitMagnitude) 4e+03 9e+03 0.1 1.0))
              (Utilities|IsValidDriverController
                (:Is%20Valid
                  (Game|Feedback|ClientStartCameraShake :self (LocalVariables|GetValidPlayerControllers) :Shake (Variables|CameraShakes|GetSmallHitCameraShake) :Scale (LocalVariables|GetCameraShakeMagnitude) :PlaySpace "CameraLocal")
                  (if (> (Variables|GetHitMagnitude) 5.5e+03)
                    (Game|Feedback|ClientPlayForceFeedback _validplayercontrollers (Variables|ForceFeedback|GetBigHitForceFeedback) "None" false false false)
                    (else
                      (Game|Feedback|ClientPlayForceFeedback _validplayercontrollers (Variables|ForceFeedback|GetSmallHitForceFeedback) "None" false false false))))))
            (:then_1
              (if (< (Math|Float|Absolute(Float) (Math|Vector|DotProduct (Math|Vector|Normalize (Vehicle|GetPreviousLinearVelocity)) "0, 0,1.000000")) 0.5)
                (bind _output_get (LocalVariables|SetHitDelta (- (Variables|Vehicle|GetCachedSpeed) (Math|Vector|VectorLength (Transformation|GetComponentVelocity (Variables|Vehicle|GetSkeletalMesh))))))
                (if (> _output_get (LocalVariables|GetCameraPushVelocityDelta))
                  (Game|Feedback|ClientStartCameraShake (LocalVariables|GetValidPlayerControllers) (Variables|CameraShakes|GetBigHitCameraShake) (Math|Float|MapRangeClamped (LocalVariables|GetHitDelta) 3e+02 9e+02 0.5 1.0) "UserDefined" (Math|Rotator|MakeRotfromX (Vehicle|GetPreviousLinearVelocity))))))))))))

(fn RandomizeNewTargetLocation
  (params (TankLocation :type Vector))
  (return (Utilities|Operators|Add TankLocation (Math|Vector|MakeVector (Math|Float|SelectFloat (Math|Random|RandomFloatinRange 5e+03 8e+03) (Math|Random|RandomFloatinRange -5e+03 -8e+03) (Math|Random|RandomBool)) (Math|Float|SelectFloat (Math|Random|RandomFloatinRange 5e+03 8e+03) (Math|Random|RandomFloatinRange -5e+03 -8e+03) (Math|Random|RandomBool)) 5e+02))))

(fn SendDamageValueToMaterials
  (params (ParameterValue :type Float%20%28double-precision%29))
  (bind _skeletalmesh (Variables|Vehicle|GetSkeletalMesh))
  (bind _healthpctmissingtoshowdmgcrinkle (Variables|VehicleDamageState|GetHealthPctMissingToShowDmgCrinkle))
  ;[1 1 1] Only Show damage crinkles after health is below a certain threshold
  (if (> ParameterValue _healthpctmissingtoshowdmgcrinkle)
    (bind _output_get (Variables|VehicleDamageState|SetSavedDamageValueForMID (Math|Float|Lerp (* (* _healthpctmissingtoshowdmgcrinkle (Variables|VehicleDamageState|GetMaxDamagedCrinkleSeverity)) -1.0) (Variables|VehicleDamageState|GetMaxDamagedCrinkleSeverity) ParameterValue)))
    (Rendering|Material|SetScalarParameterValueonMaterials _skeletalmesh "Damage" _output_get)
    (else
      (bind _output_get_2 (Variables|VehicleDamageState|SetSavedDamageValueForMID 0.0))
      (Rendering|Material|SetScalarParameterValueonMaterials _skeletalmesh "Damage" _output_get_2) ; Need to reset crinkle back to 0.0 if we're healed beyond the min threshold
    )))

(fn SetFallInstigator
  (params (Pawn :type Fort%20Player%20Pawn%20Athena%20Object%20Reference))
  (if (Vehicle|HasDriver)
    (if (not (Team|OnSameTeam (Variables|VehicleDefaults|GetDriverPawn) Pawn))
      (Class|FortPlayerPawnAthena|SetFallInstigator :self Pawn :NewFallInstigator (Game|GetInstigatorController (Variables|VehicleDefaults|GetDriverPawn))))
    (elif (Variables|VehicleImpact|GetPassengerIsInstigator)
      (Class|FortPlayerPawnAthena|SetFallInstigator :self Pawn :NewFallInstigator (Game|GetInstigatorController (Utilities|Array|Get(acopy) (Variables|VehicleDefaults|GetPassengers))))
      (else
        (Class|FortPlayerPawnAthena|SetFallInstigator :self Pawn :NewFallInstigator (Game|GetInstigatorController (Variables|VehicleDefaults|GetLastDriver)))))))

(fn SetFxSettingsForVehicleType
  (bind _ns_tank_roadandexhaust (Variables|Default|GetNS_Tank_RoadAndExhaust))
  (Niagara|SetNiagaraVariable(Vector3) _ns_tank_roadandexhaust "Wheel_Scale_Back" (Variables|FX|NiagaraSettings|GetTread_Scale_Right))
  (Niagara|SetNiagaraVariable(Vector3) _ns_tank_roadandexhaust "Wheel_Scale_Front" (Variables|FX|NiagaraSettings|GetTread_Scale_Left))
  (Effects|Components|ParticleSystem|SetFloatParameter _ns_tank_roadandexhaust "TreadWidth" (Variables|FX|GetTreadWidth))
  (Niagara|SetNiagaraVariable(Bool) _ns_tank_roadandexhaust "VehicleHasFuel" (> 0.0 0.0))
  (Effects|Components|ParticleSystem|SetColorParameter _ns_tank_roadandexhaust "ExhaustColor" (Variables|FX|GetExhaustColor))
  (Effects|Components|ParticleSystem|SetFloatParameter _ns_tank_roadandexhaust "ExhaustScale" (Variables|FX|GetExhaustScale)))

(fn SetLocalPassenger
  (params (FortPawn :type Fort%20Pawn%20Object%20Reference) (Seat :type Integer))
  (if (Pawn|IsLocallyViewed FortPawn)
    (bind _output_get (Variables|VehicleDefaults|SetIsLocalPlayerPassenger true))
    (CallFunction|LocalPassengerMatSwap)
    (Collision|SetCollisionResponsetoChannel (Variables|Vehicle|GetSkeletalMesh) "ECC_Camera" "ECR_Ignore")
    (Class|BPTankAudioController|OnPassengerStateChangeAudio (Utilities|GetAudioController) _output_get)
    (if (> Seat 1)
      (FortPawn|BindEventtoOnWeaponEquippedDelegate FortPawn (EventDispatchers|CreateEvent :!event Passenger%20Equip%20Weapon))
      (bind _asb_ranged_generic (Utilities|Casting|CastToB_Ranged_Generic (Variables|/Script/FortniteGame.FortPawn|GetCurrentWeapon FortPawn)))
      (CallFunction|PassengerEquipWeapon _asb_ranged_generic))))

(fn TurnSignalsandBrakes)

(fn ConstructionScript
  (Variables|FortAthenaVehicle|SetOverlapComponent (Variables|Default|GetOverlapVolume))
  ;[0.1077 1 0] Need new Audio Controller
  (ChildActorComponent|SetChildActorClass (Variables|Default|GetAudioControllerBP) (Variables|Audio|GetAudioController))
  (Variables|DriverCover|SetDriverCoverResetCooldownTime (ScalableFloat|GetValueatLevel (LocalVariables|GetHotfixDriverCoverResetCooldownTime)))
  (Variables|VehicleDefaults|SetDriveOverVehicleDamage (ScalableFloat|GetValueatLevel (LocalVariables|GetHotfixDriveOverVehicleDamage)))
  (Variables|VehicleExplode|SetExplodeRadius (ScalableFloat|GetValueatLevel (LocalVariables|GetHotfixExplodeRadius)))
  (Variables|VehicleExplode|SetExplodeVehicleImpulse (ScalableFloat|GetValueatLevel (LocalVariables|GetHotfixVehicleImpulse)))
  (Variables|VehicleExplode|SetExplodeSpinThrustMagnitude (ScalableFloat|GetValueatLevel (LocalVariables|GetHotfixExplodeSpinThrustMag)))
  (Variables|VehicleExplode|SetExplodeImpulseZ_Magnitude (ScalableFloat|GetValueatLevel (LocalVariables|GetHotfixExplodeImpulseZ_Mag)))
  (Variables|VehicleExplode|SetExplodePlayerImpulse (ScalableFloat|GetValueatLevel (LocalVariables|GetHotfixExplodePlayerImpulse)))
  (Variables|VehicleImpact|SetScalePlayerHitImpulseBySpeed (ScalableFloat|GetValueatLevel (LocalVariables|GetHotfixScalePlayerHitImpulseOnSpeed)))
  (Variables|VehicleImpact|SetScalePlayerHitImpulseNoDriver (ScalableFloat|GetValueatLevel (LocalVariables|GetHotfixScalePlayerHitImpulseNoDriver)))
  (Variables|TreadSkid|SetMinSpeedForTreadSkid (ScalableFloat|GetValueatLevel (LocalVariables|GetHotfixMinSpeedToTireSkid)))
  (Variables|TreadSkid|SetMinSpeedForBrakingSkid (ScalableFloat|GetValueatLevel (LocalVariables|GetHotfixMinSpeedBrakingSkid)))
  (Variables|TreadSkid|SetMinSkidAmountRequired (ScalableFloat|GetValueatLevel (LocalVariables|GetHotfixMinSkidAmountRequired)))
  (Variables|CameraShakes|SetExplosion_CameraShake_Radius (ScalableFloat|GetValueatLevel (LocalVariables|GetHotfixExplodeCameraShakeRadius)))
  (Variables|TankEngine|SetEngineOverheatDuration (ScalableFloat|GetValueatLevel (LocalVariables|GetHotfixEngineCooldownDuration)))
  (Variables|TankEngine|SetEngineMaxHealth (ScalableFloat|GetValueatLevel (LocalVariables|GetHotfixEngineMaxHealth)))
  (Variables|TankEngine|SetTankEngineCurrentHealthPercent 1.0) ; Set Health Percent of Engine to 1
  (Variables|DriverCover|SetDriverCoverMaxHealth (ScalableFloat|GetValueatLevel (LocalVariables|GetHotfixDriverCoverMaxHealth)))
  (Variables|TankEngine|SetDelayToBeginEngineRepairOnDamage (ScalableFloat|GetValueatLevel (LocalVariables|GetHotfixDelayToBeginEngineRepairOnDamage)))
  (Variables|TankEngine|SetTankEngineRepairPerTick (ScalableFloat|GetValueatLevel (LocalVariables|GetHotfixTankEngineRepairPerTick)))
  (Variables|TankEngine|SetTankEngineRepairTickRate (ScalableFloat|GetValueatLevel (LocalVariables|GetHotfixEngineRepairTickRate)))
  (Variables|DriverCover|SetDriverCoverEnabled (ScalableFloat|GetValueasBool (LocalVariables|GetHotfixDriverCoverEnabled)))
  (Variables|VehicleDefaults|SetTankSpinThrustOnCannonFire (ScalableFloat|GetValueatLevel (Variables|CannonFire|GetCannonFireSpinThrustMag)))
  (Variables|VehicleDefaults|SetShouldRockTankOnCannonFire (ScalableFloat|GetValueasBool (Variables|CannonFire|GetShouldRollTankOnCannonFire)))
  (Variables|VehicleDefaults|SetCurieEnabled (ScalableFloat|GetValueasBool (Variables|VehicleDefaults|GetHotfixCurieEnabled)))
  (Variables|VehicleDefaults|SetbRagdollEnabled (Class|AthenaFunctionLibrary|RowtoBool (Variables|VehicleDefaults|GetRagdollEnabled))))

(fn WaterSurfaceVector
  (return (Math|Vector|MakeVector (.x (Transformation|GetActorLocation)) (.y (Transformation|GetActorLocation)) (CallFunction|GetWaterLevel))))
